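# Physics asset for the dangling chains on the female medium Cavalry Alt cosmetic body, presumably used to simulate their secondary motion.
# Small capsules at the back, front and mid chain roots, plus larger capsules on the pelvis and two right-side skirt bones, form the bodies constrained to the pelvis with free rotation.
# The front chain's box and the mid chain's capsule hang from their roots with tighter, asymmetric swing and twist limits.
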PhysicsAsset /BRCosmetics/Characters/Player/Female/Medium/Bodies/F_MED_Cavalry_Alt/Meshes/F_MED_Cavalry_Alt_Physics_chain
bodies {
 dyn_chain_back_01_base_joint [{shapeType=Capsule, length=0.505, radius=0.505, shapeName=dyn_chain_back_01_base_joint_capsule, rotation=<rot 0 -90 0>}]
 dyn_chain_front_01_base_joint [{shapeType=Capsule, length=0.505, radius=0.505, shapeName=dyn_chain_front_01_base_joint_capsule, rotation=<rot 0 -90 0>}]
 dyn_chain_front_01_joint [{
   shapeType=Box
   extentX=12.3246
   extentY=3.1967
   extentZ=9.7968
   shapeName=dyn_chain_front_01_joint_box
   center=<-7.9594 0.1354 -5.5602>
   rotation=<rot 35.5226 11.0271 -25.4073>
  }]
 dyn_chain_mid_01_joint [{shapeType=Capsule, length=0.505, radius=0.505, shapeName=dyn_chain_mid_01_joint_capsule, rotation=<rot 0 -90 0>}]
 dyn_chain_mid_02_joint [{
   shapeType=Capsule
   length=7.9101
   radius=2.8109
   shapeName=dyn_chain_mid_02_joint_capsule
   center=<8.5338 2.391 0.5519>
   rotation=<rot 85.3093 -108.4907 -111.6146>
  }]
 dyn_skirt_fr_a_base_r [{
   shapeType=Capsule
   length=19.5453
   radius=6.2297
   shapeName=dyn_skirt_fr_a_base_r_capsule
   center=<10.0523 -7.1809 0>
   rotation=<rot 0 -84.8574 90>
  }]
 dyn_skirt_side_a_r [{
   shapeType=Capsule
   length=7.7948
   radius=7.1911
   shapeName=dyn_skirt_side_a_r_capsule
   center=<8.173 -7.2402 0.2716>
   rotation=<rot 88.0966 179.9991 179.9991>
  }]
 pelvis [{shapeType=Capsule, length=2.8917, radius=11.5924, center=<-6.5861 -1.3543 1.8727>, rotation=<rot 0 -90 0>}]
}
constraints [
 {
  bone1Name=dyn_chain_back_01_base_joint
  bone2Name=pelvis
  swing1LimitDegrees=45
  swing1Motion=Free
  swing2LimitDegrees=45
  swing2Motion=Free
  twistLimitDegrees=45
  twistMotion=Free
 }
 {
  bone1Name=dyn_chain_front_01_base_joint
  bone2Name=pelvis
  swing1LimitDegrees=45
  swing1Motion=Free
  swing2LimitDegrees=45
  swing2Motion=Free
  twistLimitDegrees=45
  twistMotion=Free
 }
 {
  bone1Name=dyn_chain_front_01_joint
  bone2Name=dyn_chain_front_01_base_joint
  swing1LimitDegrees=50
  swing2LimitDegrees=10
  twistLimitDegrees=10
 }
 {
  bone1Name=dyn_chain_mid_01_joint
  bone2Name=pelvis
  swing1LimitDegrees=45
  swing1Motion=Free
  swing2LimitDegrees=45
  swing2Motion=Free
  twistLimitDegrees=45
  twistMotion=Free
 }
 {bone1Name=dyn_chain_mid_02_joint, bone2Name=dyn_chain_mid_01_joint, swing1LimitDegrees=50, swing2LimitDegrees=10, twistLimitDegrees=2}
 {
  bone1Name=dyn_skirt_fr_a_base_r
  bone2Name=pelvis
  swing1LimitDegrees=45
  swing1Motion=Free
  swing2LimitDegrees=45
  swing2Motion=Free
  twistLimitDegrees=45
  twistMotion=Free
 }
 {
  bone1Name=dyn_skirt_side_a_r
  bone2Name=pelvis
  swing1LimitDegrees=45
  swing1Motion=Free
  swing2LimitDegrees=45
  swing2Motion=Free
  twistLimitDegrees=45
  twistMotion=Free
 }
]
