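# Physics asset for the front skirt of the F_MED_Cavalry_Alt female cosmetic body, with capsule bodies for skirt bones, pelvis and thighs.
# Left and right front chains run base, A, B with tight angular limits, using larger capsules on A and B.
# Thighs and all front, side and back base bones join the pelvis with free swing and twist.
# Back and side skirt bones have only tiny base capsules.
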PhysicsAsset /BRCosmetics/Characters/Player/Female/Medium/Bodies/F_MED_Cavalry_Alt/Meshes/F_MED_Cavalry_Alt_Physics_skirt_front
bodies {
 dyn_skirt_bk_a_base [{shapeType=Capsule, length=0.505, radius=0.505, shapeName=dyn_skirt_bk_a_base_capsule, rotation=<rot 0 -90 0>}]
 dyn_skirt_bk_a_base_l [{shapeType=Capsule, length=0.505, radius=0.505, shapeName=dyn_skirt_bk_a_base_l_capsule, rotation=<rot 0 -90 0>}]
 dyn_skirt_bk_a_base_r [{shapeType=Capsule, length=0.505, radius=0.505, shapeName=dyn_skirt_bk_a_base_r_capsule, rotation=<rot 0 -90 0>}]
 dyn_skirt_fr_a_base [{shapeType=Capsule, length=0.505, radius=0.505, shapeName=dyn_skirt_fr_a_base_capsule, rotation=<rot 0 -90 0>}]
 dyn_skirt_fr_a_base_l [{shapeType=Capsule, length=0.505, radius=0.505, shapeName=dyn_skirt_fr_a_base_l_capsule, rotation=<rot 0 -90 0>}]
 dyn_skirt_fr_a_base_r [{shapeType=Capsule, length=0.505, radius=0.505, shapeName=dyn_skirt_fr_a_base_r_capsule, rotation=<rot 0 -90 0>}]
 dyn_skirt_fr_a_l [{
   shapeType=Capsule
   length=2.5
   radius=5
   shapeName=dyn_skirt_fr_a_l_capsule
   center=<-7.0786 -6.5817 0.002>
   rotation=<rot 0.3413 -96.6485 -89.3844>
  }]
 dyn_skirt_fr_a_r [{shapeType=Capsule, length=2.5, radius=5, shapeName=dyn_skirt_fr_a_r_capsule, center=<7 4 0>, rotation=<rot 0 -91.6567 -88.0253>}]
 dyn_skirt_fr_b_l [
  {
   shapeType=Capsule
   length=6.9008
   radius=6.5992
   shapeName=dyn_skirt_fr_b_l_capsule
   center=<-6.5627 -3.2677 0.1273>
   rotation=<rot -1.6126 -99.2437 87.6761>
  }
  {
   shapeType=Capsule
   length=3.7648
   radius=4.9357
   shapeName=dyn_skirt_fr_b_l_capsule
   center=<-3.5662 2.3133 7.643>
   rotation=<rot -88.2455 -0.0161 -8.3132>
  }
 ]
 dyn_skirt_fr_b_r [{shapeType=Capsule, length=5.5, radius=8, shapeName=dyn_skirt_fr_b_r_capsule, center=<7 4 0>, rotation=<rot -21.1149 -89.9999 -88.5525>}]
 dyn_skirt_side_a_base_l [{shapeType=Capsule, length=0.505, radius=0.505, shapeName=dyn_skirt_side_a_base_l_capsule, rotation=<rot 0 -90 0>}]
 dyn_skirt_side_a_base_r [{shapeType=Capsule, length=0.505, radius=0.505, shapeName=dyn_skirt_side_a_base_r_capsule, rotation=<rot 0 -90 0>}]
 pelvis [{shapeType=Capsule, length=8.1069, radius=9.8036, center=<-8.3424 -1.3543 0>, rotation=<rot 0 -90 0>}]
 thigh_l [{shapeType=Capsule, length=27.75, radius=7.25, shapeName=thigh_l_capsule, center=<-16.9587 0.7457 0.6245>, rotation=<rot 0 -86.5304 -92.7649>}]
 thigh_r [{shapeType=Capsule, length=27.75, radius=7.25, shapeName=thigh_l_capsule, center=<16.9587 -0.7457 -0.6245>, rotation=<rot 0 -86.5304 87.2351>}]
}
constraints [
 {
  bone1Name=thigh_r
  bone2Name=pelvis
  swing1LimitDegrees=45
  swing1Motion=Free
  swing2LimitDegrees=45
  swing2Motion=Free
  twistLimitDegrees=45
  twistMotion=Free
 }
 {
  bone1Name=dyn_skirt_fr_a_base_l
  bone2Name=pelvis
  swing1LimitDegrees=45
  swing1Motion=Free
  swing2LimitDegrees=45
  swing2Motion=Free
  twistLimitDegrees=45
  twistMotion=Free
 }
 {
  bone1Name=dyn_skirt_fr_a_base_r
  bone2Name=pelvis
  swing1LimitDegrees=45
  swing1Motion=Free
  swing2LimitDegrees=45
  swing2Motion=Free
  twistLimitDegrees=45
  twistMotion=Free
 }
 {bone1Name=dyn_skirt_fr_b_r, bone2Name=dyn_skirt_fr_a_r, swing1LimitDegrees=20, swing2LimitDegrees=5, twistLimitDegrees=5}
 {bone1Name=dyn_skirt_fr_a_r, bone2Name=dyn_skirt_fr_a_base_r, swing1LimitDegrees=20, swing2LimitDegrees=5, twistLimitDegrees=5}
 {bone1Name=dyn_skirt_fr_b_l, bone2Name=dyn_skirt_fr_a_l, swing1LimitDegrees=20, swing2LimitDegrees=5, twistLimitDegrees=5}
 {bone1Name=dyn_skirt_fr_a_l, bone2Name=dyn_skirt_fr_a_base_l, swing1LimitDegrees=20, swing2LimitDegrees=5, twistLimitDegrees=5}
 {
  bone1Name=dyn_skirt_bk_a_base
  bone2Name=pelvis
  swing1LimitDegrees=45
  swing1Motion=Free
  swing2LimitDegrees=45
  swing2Motion=Free
  twistLimitDegrees=45
  twistMotion=Free
 }
 {
  bone1Name=dyn_skirt_bk_a_base_l
  bone2Name=pelvis
  swing1LimitDegrees=45
  swing1Motion=Free
  swing2LimitDegrees=45
  swing2Motion=Free
  twistLimitDegrees=45
  twistMotion=Free
 }
 {
  bone1Name=dyn_skirt_bk_a_base_r
  bone2Name=pelvis
  swing1LimitDegrees=45
  swing1Motion=Free
  swing2LimitDegrees=45
  swing2Motion=Free
  twistLimitDegrees=45
  twistMotion=Free
 }
 {
  bone1Name=dyn_skirt_fr_a_base
  bone2Name=pelvis
  swing1LimitDegrees=45
  swing1Motion=Free
  swing2LimitDegrees=45
  swing2Motion=Free
  twistLimitDegrees=45
  twistMotion=Free
 }
 {
  bone1Name=dyn_skirt_side_a_base_l
  bone2Name=pelvis
  swing1LimitDegrees=45
  swing1Motion=Free
  swing2LimitDegrees=45
  swing2Motion=Free
  twistLimitDegrees=45
  twistMotion=Free
 }
 {
  bone1Name=dyn_skirt_side_a_base_r
  bone2Name=pelvis
  swing1LimitDegrees=45
  swing1Motion=Free
  swing2LimitDegrees=45
  swing2Motion=Free
  twistLimitDegrees=45
  twistMotion=Free
 }
 {
  bone1Name=thigh_l
  bone2Name=pelvis
  swing1LimitDegrees=45
  swing1Motion=Free
  swing2LimitDegrees=45
  swing2Motion=Free
  twistLimitDegrees=45
  twistMotion=Free
 }
]
