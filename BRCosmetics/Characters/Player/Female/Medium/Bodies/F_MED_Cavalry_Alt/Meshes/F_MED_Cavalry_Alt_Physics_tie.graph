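# Physics asset for the tie and bracelet bones on the female medium Cavalry Alt player body.
# The tie is a six-capsule chain hanging from spine_05, with tight twist and swing limits that narrow toward the tip on one axis.
# Three bracelet pieces each pair a small base, freely rotating against spine_05, with a tightly limited capsule.
# Only spine_03 to spine_05 represent the body, as large torso capsules.
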PhysicsAsset /BRCosmetics/Characters/Player/Female/Medium/Bodies/F_MED_Cavalry_Alt/Meshes/F_MED_Cavalry_Alt_Physics_tie
bodies {
 dyn_bracelet_bottom [{shapeType=Capsule, length=6.4375, radius=0.6562, shapeName=dyn_bracelet_bottom_capsule}]
 dyn_bracelet_bottom_2 [{shapeType=Capsule, length=2.2188, radius=0.75, shapeName=dyn_bracelet_bottom_2_capsule}]
 dyn_bracelet_bottom_2_base [{shapeType=Capsule, length=0.505, radius=0.505, shapeName=dyn_bracelet_bottom_2_base_capsule, rotation=<rot 0 -90 0>}]
 dyn_bracelet_bottom_base [{shapeType=Capsule, length=0.505, radius=0.505, shapeName=dyn_bracelet_bottom_base_capsule, rotation=<rot 0 -90 0>}]
 dyn_bracelet_top [{shapeType=Capsule, length=5.9375, radius=0.6875, shapeName=dyn_bracelet_top_capsule}]
 dyn_bracelet_top_base [{shapeType=Capsule, length=0.505, radius=0.505, shapeName=dyn_bracelet_top_base_capsule, rotation=<rot 0 -90 0>}]
 dyn_tie_00_joint [{shapeType=Capsule, length=2.3429, radius=1.4824, shapeName=dyn_tie_00_joint_capsule, center=<2.6597 0 0>, rotation=<rot 0 90.0001 0>}]
 dyn_tie_01_joint [{
   shapeType=Capsule
   length=1.5353
   radius=1.7122
   shapeName=dyn_tie_01_joint_capsule
   center=<2.6333 0.1391 -0.0094>
   rotation=<rot 0 -92.6536 -89.9004>
  }]
 dyn_tie_02_joint [{
   shapeType=Capsule
   length=1.2457
   radius=2.0903
   shapeName=dyn_tie_02_joint_capsule
   center=<2.9388 0.0237 -0.0085>
   rotation=<rot 0 -91.7808 -89.5572>
  }]
 dyn_tie_03_joint [{
   shapeType=Capsule
   length=0.8791
   radius=2.2764
   shapeName=dyn_tie_03_joint_capsule
   center=<2.8888 -0.0294 -0.0301>
   rotation=<rot 0.0104 -89.9999 -89.4573>
  }]
 dyn_tie_04_joint [{
   shapeType=Capsule
   length=0.711
   radius=2.0078
   shapeName=dyn_tie_04_joint_capsule
   center=<2.5256 -0.0083 -0.025>
   rotation=<rot 0 -90.7352 -89.2667>
  }]
 dyn_tie_root_joint [{shapeType=Capsule, length=0.8626, radius=1.8734, shapeName=dyn_tie_root_joint_capsule, center=<1.7031 -0.0905 0>, rotation=<rot 0 -4.5142 0>}]
 spine_03 [{shapeType=Capsule, length=6.9698, radius=8.5138, center=<0.0133 2.3575 0.102>, rotation=<rot -0.4483 -90 0.5222>}]
 spine_04 [{
   shapeType=Capsule
   length=12.8079
   radius=11.555
   shapeName=spine_04_capsule
   center=<-1.1401 0.5411 0.0364>
   rotation=<rot -0.4398 -90 0.5187>
  }]
 spine_05 [{shapeType=Capsule, length=15.7615, radius=7.2978, center=<0.0071 -0.1763 0.142>, rotation=<rot 0.5852 -177.3306 1.5418>}]
}
constraints [
 {bone1Name=spine_05, bone2Name=spine_04, swing1LimitDegrees=45, swing2LimitDegrees=45, twistLimitDegrees=45}
 {bone1Name=dyn_tie_01_joint, bone2Name=dyn_tie_00_joint, swing1LimitDegrees=20, swing2LimitDegrees=20, twistLimitDegrees=5}
 {bone1Name=dyn_tie_02_joint, bone2Name=dyn_tie_01_joint, swing1LimitDegrees=20, swing2LimitDegrees=15, twistLimitDegrees=5}
 {bone1Name=dyn_tie_03_joint, bone2Name=dyn_tie_02_joint, swing1LimitDegrees=20, swing2LimitDegrees=10, twistLimitDegrees=5}
 {bone1Name=dyn_tie_04_joint, bone2Name=dyn_tie_03_joint, swing1LimitDegrees=20, swing2LimitDegrees=5, twistLimitDegrees=5}
 {
  bone1Name=spine_04
  bone2Name=spine_03
  swing1LimitDegrees=45
  swing1Motion=Free
  swing2LimitDegrees=45
  swing2Motion=Free
  twistLimitDegrees=45
  twistMotion=Free
 }
 {bone1Name=dyn_tie_root_joint, bone2Name=spine_05, swing1LimitDegrees=20, swing2LimitDegrees=20, twistLimitDegrees=5}
 {bone1Name=dyn_tie_00_joint, bone2Name=dyn_tie_root_joint, swing1LimitDegrees=20, swing2LimitDegrees=20, twistLimitDegrees=5}
 {
  bone1Name=dyn_bracelet_bottom_base
  bone2Name=spine_05
  swing1LimitDegrees=45
  swing1Motion=Free
  swing2LimitDegrees=45
  swing2Motion=Free
  twistLimitDegrees=45
  twistMotion=Free
 }
 {bone1Name=dyn_bracelet_bottom, bone2Name=dyn_bracelet_bottom_base, swing1LimitDegrees=5, swing2LimitDegrees=5, twistLimitDegrees=3}
 {
  bone1Name=dyn_bracelet_top_base
  bone2Name=spine_05
  swing1LimitDegrees=10
  swing1Motion=Free
  swing2LimitDegrees=10
  swing2Motion=Free
  twistLimitDegrees=3
  twistMotion=Free
 }
 {bone1Name=dyn_bracelet_top, bone2Name=dyn_bracelet_top_base, swing1LimitDegrees=5, swing2LimitDegrees=5, twistLimitDegrees=3}
 {
  bone1Name=dyn_bracelet_bottom_2_base
  bone2Name=spine_05
  swing1LimitDegrees=45
  swing1Motion=Free
  swing2LimitDegrees=45
  swing2Motion=Free
  twistLimitDegrees=45
  twistMotion=Free
 }
 {bone1Name=dyn_bracelet_bottom_2, bone2Name=dyn_bracelet_bottom_2_base, swing1LimitDegrees=3, swing2LimitDegrees=3, twistLimitDegrees=3}
]
