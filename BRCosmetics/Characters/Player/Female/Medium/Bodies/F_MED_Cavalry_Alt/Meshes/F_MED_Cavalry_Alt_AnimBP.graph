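# Animation Blueprint for the F_MED_Cavalry_Alt female body cosmetic, layering secondary motion and pose fixes onto the pose copied from the parent mesh.
# Chained constraints and additive bone tweaks adjust collar, lapel, skirt and chains, around RigidBody simulations for chain, tie, skirt and falling-only front skirt.
# A final Flap node drives the bracelet from hand rotation.
# The update event derives blend alphas and flags from thigh swing angles and movement state.
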AnimBlueprint /BRCosmetics/Characters/Player/Female/Medium/Bodies/F_MED_Cavalry_Alt/Meshes/F_MED_Cavalry_Alt_AnimBP
target_skeleton=&/F_MED_Cavalry_Alt_Skeleton
graphs [
 AnimGraph {nodes [
   n1 = "Transform (Modify) Bone - Bone: scalePelvisAlt_joint"!(
    ComponentPose=Local To Component(LocalPose=Copy Pose From Mesh() {bUseAttachedParent=true})
    bAlphaBoolEnabled:bool=Get LOD_2_3() -> (LOD_2_3:bool)
    Translation="0,1,0"
    Rotation="0,0,0"
    Scale="0.7,0.7,0.7"
   ) -> (Pose) {
    alphaInputType=Bool
    rotationMode=BMM_Additive
    rotationSpace=BCS_BoneSpace
    scaleMode=BMM_Additive
    scaleSpace=BCS_BoneSpace
    translationMode=BMM_Additive
    translationSpace=BCS_BoneSpace
    alphaBoolBlend {blendInTime=0.2, blendOption=Linear, blendOutTime=0.2}
    boneToModify {boneName=scalePelvisAlt_joint}
    scale=<0.7 0.7 0.7>
    translation {y=1}
   }
   n2 = "Transform (Modify) Bone - Bone: extra_back_pelvis_m"(ComponentPose=@n1, Alpha=1, Translation="0,1,0", Rotation="0,0,0", Scale="1,1,1") {
    rotationMode=BMM_Additive
    rotationSpace=BCS_BoneSpace
    translationMode=BMM_Additive
    translationSpace=BCS_BoneSpace
    alphaBoolBlend {blendInTime=0.2, blendOption=Linear, blendOutTime=0.2}
    boneToModify {boneName=extra_back_pelvis_m}
    translation {y=1}
   }
   n3 = "Constraint - lapel_r_joint (PP)"!(ComponentPose=@n2, Alpha=1, ConstraintWeights_0=0.5018, ConstraintWeights_1:float=0.4982) -> (Pose) {
    alphaBoolBlend {blendOption=Linear}
    boneToModify {boneName=lapel_r_joint}
    constraintSetup [{targetBone {boneName=spine_05}}, {targetBone {boneName=spine_04}}]
    constraintWeights [1,1]
   }
   n4 = Get collar_r_bool_var() -> (collar_r_bool_var:bool)
   n5 = "Transform (Modify) Bone - Bone: collar_l_joint"!(ComponentPose=@n3, bAlphaBoolEnabled:bool=@n4, Translation="0,0,0", Rotation="17,0,0", Scale="1,1,1") -> (Pose) {
    alphaInputType=Bool
    rotationMode=BMM_Additive
    rotationSpace=BCS_BoneSpace
    translationMode=BMM_Additive
    translationSpace=BCS_BoneSpace
    alphaBoolBlend {blendInTime=0.2, blendOption=Linear, blendOutTime=0.2}
    boneToModify {boneName=collar_l_joint}
    rotation {pitch=17}
   }
   n6 = "Transform (Modify) Bone - Bone: collar_r_joint"!(ComponentPose=@n5, bAlphaBoolEnabled:bool=@n4, Translation="0,0,0", Rotation="-6,0,0", Scale="1,1,1") -> (Pose) {
    alphaInputType=Bool
    rotationMode=BMM_Additive
    rotationSpace=BCS_BoneSpace
    translationMode=BMM_Additive
    translationSpace=BCS_BoneSpace
    alphaBoolBlend {blendInTime=0.2, blendOption=Linear, blendOutTime=0.2}
    boneToModify {boneName=collar_r_joint}
    rotation {pitch=-6}
   }
   n7 = "Transform (Modify) Bone - Bone: collar_l_joint"!(
    ComponentPose=@n6
    Alpha:float=Get L_aim_var_float() -> (L_aim_var_float:double)
    Translation="-1.5,0.4,1.5"
    Rotation="0.001,0,8.304"
    Scale="1,1,1"
   ) -> (Pose) {
    rotationMode=BMM_Additive
    rotationSpace=BCS_BoneSpace
    translationMode=BMM_Additive
    translationSpace=BCS_BoneSpace
    alphaBoolBlend {blendInTime=0.1, blendOption=Linear, blendOutTime=0.3}
    boneToModify {boneName=collar_l_joint}
    rotation {pitch=0.001, roll=8.304}
    translation=<-1.5 0.4 1.5>
   }
   n8 = "Transform (Modify) Bone - Bone: collar_r_joint"(
    ComponentPose=@n7
    bAlphaBoolEnabled=Get R_aim_var_bool() -> (R_aim_var_bool:bool)
    Translation="0.182,0.271,-1.778"
    Rotation="2.210,14.911,-6.994"
    Scale="1,1,1"
   ) {
    alphaInputType=Bool
    rotationMode=BMM_Additive
    rotationSpace=BCS_BoneSpace
    translationMode=BMM_Additive
    translationSpace=BCS_BoneSpace
    alphaBoolBlend {blendInTime=0.1, blendOption=Linear, blendOutTime=0.3}
    boneToModify {boneName=collar_r_joint}
    rotation=<rot 2.21 14.911 -6.994>
    translation=<0.182 0.271 -1.778>
   }
   n9 = Get bIsParachuteOpened()
   n10 = "Transform (Modify) Bone - Bone: collar_l_joint"!(ComponentPose=@n8, bAlphaBoolEnabled:bool=@n9, Translation="-3,0.4,1.5", Rotation="0.001,0,8.304", Scale="1,1,1") -> (Pose) {
    alphaInputType=Bool
    rotationMode=BMM_Additive
    rotationSpace=BCS_BoneSpace
    translationMode=BMM_Additive
    translationSpace=BCS_BoneSpace
    alphaBoolBlend {blendInTime=0.2, blendOption=Linear, blendOutTime=0.2}
    boneToModify {boneName=collar_l_joint}
    rotation {pitch=0.001, roll=8.304}
    translation=<-3 0.4 1.5>
   }
   n11 = "Transform (Modify) Bone - Bone: collar_r_joint"(ComponentPose=@n10, bAlphaBoolEnabled=@n9, Translation="-2,0.271,-0.5", Rotation="2.210,14,-6.994", Scale="1,1,1") {
    alphaInputType=Bool
    rotationMode=BMM_Additive
    rotationSpace=BCS_BoneSpace
    translationMode=BMM_Additive
    translationSpace=BCS_BoneSpace
    alphaBoolBlend {blendInTime=0.2, blendOption=Linear, blendOutTime=0.2}
    boneToModify {boneName=collar_r_joint}
    rotation=<rot 2.21 14 -6.994>
    translation=<-2 0.271 -0.5>
   }
   n12 = "Transform (Modify) Bone - Bone: collar_r_joint"!(
    ComponentPose=@n11
    Alpha:float=Get upperarm_r_var() -> (upperarm_r_var:double)
    Translation="0,0,0"
    Rotation="-70.000,0,0"
    Scale_X:double=Get scale_collar_r() -> (scale_collar_r:double)
    Scale_Y:double=1
    Scale_Z:double=1
   ) -> (Pose) {
    rotationMode=BMM_Additive
    rotationSpace=BCS_BoneSpace
    scaleMode=BMM_Additive
    scaleSpace=BCS_BoneSpace
    translationMode=BMM_Additive
    translationSpace=BCS_BoneSpace
    alphaBoolBlend {blendInTime=0.2, blendOption=Linear, blendOutTime=0.2}
    boneToModify {boneName=collar_r_joint}
    rotation {pitch=-70}
    scale {x=-0.8}
   }
   n13 = "Transform (Modify) Bone - Bone: lapel_r_joint"!(ComponentPose=@n12, bAlphaBoolEnabled:bool=Get R_aim_var_bool(), Translation="0,0,0", Rotation="-5.000,0,0", Scale="0.658,1.000,1.000") -> (Pose) {
    alphaInputType=Bool
    rotationMode=BMM_Additive
    rotationSpace=BCS_BoneSpace
    scaleMode=BMM_Additive
    scaleSpace=BCS_BoneSpace
    translationMode=BMM_Additive
    translationSpace=BCS_BoneSpace
    alphaBoolBlend {blendInTime=0.2, blendOption=Linear, blendOutTime=0.2}
    boneToModify {boneName=lapel_r_joint}
    rotation {pitch=-5}
    scale {x=0.658}
   }
   n14 = "Constraint - dyn_skirt_fr_a_base_l (RR)"!(ComponentPose=@n13, Alpha=1, ConstraintWeights_0=0.1, ConstraintWeights_1:float=0.9) -> (Pose) {
    alphaBoolBlend {blendOption=Linear}
    boneToModify {boneName=dyn_skirt_fr_a_base_l}
    constraintSetup [{transformType=Rotation, targetBone {boneName=pelvis}}, {transformType=Rotation, targetBone {boneName=thigh_l}}]
    constraintWeights [1,1]
   }
   n15 = "Constraint - dyn_skirt_side_a_base_l (RR)"!(ComponentPose=@n14, Alpha=1, ConstraintWeights_0=0.4, ConstraintWeights_1:float=0.6) -> (Pose) {
    alphaBoolBlend {blendOption=Linear}
    boneToModify {boneName=dyn_skirt_side_a_base_l}
    constraintSetup [{transformType=Rotation, targetBone {boneName=pelvis}}, {transformType=Rotation, targetBone {boneName=thigh_l}}]
    constraintWeights [1,1]
   }
   n16 = "Constraint - dyn_skirt_bk_a_base_l (RR)"!(ComponentPose=@n15, Alpha=1, ConstraintWeights_0=0.9, ConstraintWeights_1:float=0.1) -> (Pose) {
    alphaBoolBlend {blendOption=Linear}
    boneToModify {boneName=dyn_skirt_bk_a_base_l}
    constraintSetup [{transformType=Rotation, targetBone {boneName=pelvis}}, {transformType=Rotation, targetBone {boneName=thigh_l}}]
    constraintWeights [1,1]
   }
   n17 = "Constraint - dyn_skirt_fr_a_base_r (RR)"!(ComponentPose=@n16, Alpha=1, ConstraintWeights_0=0.1, ConstraintWeights_1:float=0.9) -> (Pose) {
    alphaBoolBlend {blendOption=Linear}
    boneToModify {boneName=dyn_skirt_fr_a_base_r}
    constraintSetup [{transformType=Rotation, targetBone {boneName=pelvis}}, {transformType=Rotation, targetBone {boneName=thigh_r}}]
    constraintWeights [1,1]
   }
   n18 = "Constraint - dyn_skirt_side_a_base_r (RR)"!(ComponentPose=@n17, Alpha=1, ConstraintWeights_0=0.4, ConstraintWeights_1:float=0.6) -> (Pose) {
    alphaBoolBlend {blendOption=Linear}
    boneToModify {boneName=dyn_skirt_side_a_base_r}
    constraintSetup [{transformType=Rotation, targetBone {boneName=pelvis}}, {transformType=Rotation, targetBone {boneName=thigh_r}}]
    constraintWeights [1,1]
   }
   n19 = "Constraint - dyn_skirt_bk_a_base_r (RR)"!(ComponentPose=@n18, Alpha=1, ConstraintWeights_0=0.9, ConstraintWeights_1:float=0.1) -> (Pose) {
    alphaBoolBlend {blendOption=Linear}
    boneToModify {boneName=dyn_skirt_bk_a_base_r}
    constraintSetup [{transformType=Rotation, targetBone {boneName=pelvis}}, {transformType=Rotation, targetBone {boneName=thigh_r}}]
    constraintWeights [1,1]
   }
   n20 = Get skirt_fr_a_base_l_var() -> (skirt_fr_a_base_l_var:double)
   n21 = "Transform (Modify) Bone - Bone: extra_front_thigh_l"(ComponentPose=@n19, Alpha=@n20, Translation="0,-20,0", Rotation="0,0,0", Scale="1,1,1") {
    rotationMode=BMM_Additive
    rotationSpace=BCS_BoneSpace
    translationMode=BMM_Additive
    translationSpace=BCS_BoneSpace
    alphaBoolBlend {blendOption=Linear}
    boneToModify {boneName=extra_front_thigh_l}
    translation {y=-20}
   }
   n22 = "Transform (Modify) Bone - Bone: extra_back_thigh_l"(ComponentPose=@n21, Alpha=@n20, Translation="24.5,0,0", Rotation="0,0,0", Scale="1,1,1") {
    rotationMode=BMM_Additive
    rotationSpace=BCS_BoneSpace
    translationMode=BMM_Additive
    translationSpace=BCS_BoneSpace
    alphaBoolBlend {blendOption=Linear}
    boneToModify {boneName=extra_back_thigh_l}
    translation {x=24.5}
   }
   n23 = "Transform (Modify) Bone - Bone: dyn_skirt_fr_a_base_l"(ComponentPose=@n22, Alpha=@n20, Translation="4.514,-9.095,-2.841", Rotation="-13.989,11.753,-18.532", Scale="1,1,1") {
    rotationMode=BMM_Additive
    rotationSpace=BCS_BoneSpace
    translationMode=BMM_Additive
    translationSpace=BCS_BoneSpace
    alphaBoolBlend {blendOption=Linear}
    boneToModify {boneName=dyn_skirt_fr_a_base_l}
    rotation=<rot -13.989 11.753 -18.532>
    translation=<4.514 -9.095 -2.841>
   }
   n24 = "Transform (Modify) Bone - Bone: dyn_skirt_bk_a_base_l"(ComponentPose=@n23, Alpha=@n20, Translation="0,0,0", Rotation="-15,0,0.028", Scale="1,1,1") {
    rotationMode=BMM_Additive
    rotationSpace=BCS_BoneSpace
    translationMode=BMM_Additive
    translationSpace=BCS_BoneSpace
    alphaBoolBlend {blendOption=Linear}
    boneToModify {boneName=dyn_skirt_bk_a_base_l}
    rotation {pitch=-15, roll=0.028}
   }
   n25 = "Transform (Modify) Bone - Bone: dyn_skirt_side_a_base_l"(ComponentPose=@n24, Alpha=@n20, Translation="0,0,0", Rotation="-4.968,-48.100,-9.176", Scale="1,1,1") {
    rotationMode=BMM_Additive
    rotationSpace=BCS_BoneSpace
    translationMode=BMM_Additive
    translationSpace=BCS_BoneSpace
    alphaBoolBlend {blendOption=Linear}
    boneToModify {boneName=dyn_skirt_side_a_base_l}
    rotation=<rot -4.968 -48.1 -9.176>
   }
   n26 = Get!() -> (dyn_skirt_fr_a_base_r_Var_Location:Vector, dyn_skirt_fr_a_base_r_Var_Rotation:Rotator, dyn_skirt_fr_a_base_r_Var_Scale:Vector)
   n27 = Get skirt_fr_a_base_r_var() -> (skirt_fr_a_base_r_var:double)
   n28 = "Transform (Modify) Bone - Bone: dyn_skirt_fr_a_base_r"(ComponentPose=@n25, Alpha=@n27, Translation=@n26, Rotation=@n26.dyn_skirt_fr_a_base_r_Var_Rotation, Scale="1,1,1") {
    rotationMode=BMM_Additive
    rotationSpace=BCS_BoneSpace
    translationMode=BMM_Additive
    translationSpace=BCS_BoneSpace
    alphaBoolBlend {blendOption=Linear}
    boneToModify {boneName=dyn_skirt_fr_a_base_r}
    rotation=<rot 2.705 10.75 -9.391>
    translation=<-1.555 3.538 2.405>
   }
   n29 = "Transform (Modify) Bone - Bone: dyn_skirt_side_a_base_r"(ComponentPose=@n28, Alpha=@n27, Translation="0,0,0", Rotation="23.5,16,7", Scale="1,1,1") {
    rotationMode=BMM_Additive
    rotationSpace=BCS_BoneSpace
    translationMode=BMM_Additive
    translationSpace=BCS_BoneSpace
    alphaBoolBlend {blendOption=Linear}
    boneToModify {boneName=dyn_skirt_side_a_base_r}
    rotation=<rot 23.5 16 7>
   }
   n30 = "Transform (Modify) Bone - Bone: dyn_skirt_bk_a_base_r"(ComponentPose=@n29, Alpha=@n27, Translation="0,0,0", Rotation="-15,-0.002,0.000", Scale="1,1,1") {
    rotationMode=BMM_Additive
    rotationSpace=BCS_BoneSpace
    translationMode=BMM_Additive
    translationSpace=BCS_BoneSpace
    alphaBoolBlend {blendOption=Linear}
    boneToModify {boneName=dyn_skirt_bk_a_base_r}
    rotation {pitch=-15, yaw=-0.002}
   }
   n31 = "Transform (Modify) Bone - Bone: extra_front_thigh_r"(ComponentPose=@n30, Alpha=@n27, Translation="0,20,0", Rotation="0,0,0", Scale="1,1,1") {
    rotationMode=BMM_Additive
    rotationSpace=BCS_BoneSpace
    translationMode=BMM_Additive
    translationSpace=BCS_BoneSpace
    alphaBoolBlend {blendOption=Linear}
    boneToModify {boneName=extra_front_thigh_r}
    translation {y=20}
   }
   n32 = "Transform (Modify) Bone - Bone: dyn_skirt_bk_a_base_l"!(ComponentPose=@n31, Alpha=Get L_swinging_back_skirt() -> (L_swinging_back_skirt:double), Rotation="-7,0,0") -> (Pose) {
    rotationMode=BMM_Additive
    rotationSpace=BCS_BoneSpace
    translationMode=BMM_Additive
    translationSpace=BCS_BoneSpace
    alphaBoolBlend {blendOption=Linear}
    boneToModify {boneName=dyn_skirt_bk_a_base_l}
    rotation {pitch=-7}
   }
   n33 = "Transform (Modify) Bone - Bone: dyn_skirt_bk_a_base"!(ComponentPose=@n32, Alpha=Get L_swinging_back_skirt(), Rotation="12,0,0") -> (Pose) {
    rotationMode=BMM_Additive
    rotationSpace=BCS_BoneSpace
    translationMode=BMM_Additive
    translationSpace=BCS_BoneSpace
    alphaBoolBlend {blendOption=Linear}
    boneToModify {boneName=dyn_skirt_bk_a_base}
    rotation {pitch=12}
   }
   n34 = "Transform (Modify) Bone - Bone: dyn_skirt_bk_a_base_r"!(ComponentPose=@n33, Alpha=Get L_swinging_back_skirt(), Rotation="7,0,0") -> (Pose) {
    rotationMode=BMM_Additive
    rotationSpace=BCS_BoneSpace
    translationMode=BMM_Additive
    translationSpace=BCS_BoneSpace
    alphaBoolBlend {blendOption=Linear}
    boneToModify {boneName=dyn_skirt_bk_a_base_r}
    rotation {pitch=7}
   }
   n35 = "Transform (Modify) Bone - Bone: dyn_skirt_bk_a_base_l"(ComponentPose=@n34, Alpha=Get R_swinging_back_skirt() -> (R_swinging_back_skirt:double), Rotation="7,0,0") {
    rotationMode=BMM_Additive
    rotationSpace=BCS_BoneSpace
    translationMode=BMM_Additive
    translationSpace=BCS_BoneSpace
    alphaBoolBlend {blendOption=Linear}
    boneToModify {boneName=dyn_skirt_bk_a_base_l}
    rotation {pitch=7}
   }
   n36 = "Transform (Modify) Bone - Bone: dyn_skirt_bk_a_base"(ComponentPose=@n35, Alpha=Get R_swinging_back_skirt(), Rotation="-12,0,0.000") {
    rotationMode=BMM_Additive
    rotationSpace=BCS_BoneSpace
    translationMode=BMM_Additive
    translationSpace=BCS_BoneSpace
    alphaBoolBlend {blendOption=Linear}
    boneToModify {boneName=dyn_skirt_bk_a_base}
    rotation {pitch=-12}
   }
   n37 = "Transform (Modify) Bone - Bone: dyn_skirt_bk_a_base_r"(ComponentPose=@n36, Alpha=Get R_swinging_back_skirt(), Rotation="-7,0,0") {
    rotationMode=BMM_Additive
    rotationSpace=BCS_BoneSpace
    translationMode=BMM_Additive
    translationSpace=BCS_BoneSpace
    alphaBoolBlend {blendOption=Linear}
    boneToModify {boneName=dyn_skirt_bk_a_base_r}
    rotation {pitch=-7}
   }
   n38 = "Transform (Modify) Bone - Bone: extra_front_thigh_r"!(
    ComponentPose=@n37
    bAlphaBoolEnabled:bool=Get R_thigh volume() -> (R_thigh volume:bool)
    Translation="0,-15,0"
    Rotation="0,0,0"
    Scale="1,1,1"
   ) -> (Pose) {
    alphaInputType=Bool
    rotationMode=BMM_Additive
    rotationSpace=BCS_BoneSpace
    translationMode=BMM_Additive
    translationSpace=BCS_BoneSpace
    alphaBoolBlend {blendInTime=0.2, blendOption=Linear, blendOutTime=0.1}
    boneToModify {boneName=extra_front_thigh_r}
    translation {y=-15}
   }
   n39 = "Constraint - dyn_skirt_fr_a (PP)"!(ComponentPose=@n38, Alpha=1, ConstraintWeights_0=0.5, ConstraintWeights_1:float=0.5) -> (Pose) {
    alphaBoolBlend {blendOption=Linear}
    boneToModify {boneName=dyn_skirt_fr_a}
    constraintSetup [{targetBone {boneName=dyn_skirt_fr_a_l}}, {targetBone {boneName=dyn_skirt_fr_a_r}}]
    constraintWeights [1,1]
   }
   n40 = "Constraint - dyn_chain_front_01_base_joint (RR)"!(ComponentPose=@n39, Alpha=1, ConstraintWeights_0=0.1, ConstraintWeights_1:float=0.9) -> (Pose) {
    alphaBoolBlend {blendOption=Linear}
    boneToModify {boneName=dyn_chain_front_01_base_joint}
    constraintSetup [{transformType=Rotation, targetBone {boneName=pelvis}}, {transformType=Rotation, targetBone {boneName=dyn_skirt_fr_a_base_r}}]
    constraintWeights [1,1]
   }
   n41 = "Constraint - dyn_chain_mid_01_joint (PP)"!(
    ComponentPose=@n40
    Alpha=Get dyn_chain_01_base_var() -> (dyn_chain_01_base_var:double)
    ConstraintWeights_0=0.5
    ConstraintWeights_1:float=0.5
   ) -> (Pose) {
    alphaBoolBlend {blendOption=Linear}
    boneToModify {boneName=dyn_chain_mid_01_joint}
    constraintSetup [{targetBone {boneName=pelvis}}, {targetBone {boneName=dyn_skirt_side_a_base_r}}]
    constraintWeights [1,1]
   }
   n42 = RigidBody(ComponentPose=@n41) {
    bOverrideWorldGravity=true
    overridePhysicsAsset=&/F_MED_Cavalry_Alt_Physics_chain
    simulationSpace=BaseBoneSpace
    alphaBoolBlend {blendOption=Linear}
    baseBoneRef {boneName=pelvis}
    componentAppliedLinearAccClamp=<1000 1000 1000>
    componentLinearAccScale=<0.5 0.5 0.5>
    componentLinearVelScale=<0.5 0.5 0.5>
    overrideWorldGravity {z=-700}
    simSpaceSettings {
     maxAngularAcceleration=500
     maxAngularVelocity=3
     maxLinearAcceleration=400
     maxLinearVelocity=250
     velocityScaleZ=0.35
     externalLinearDragV {z=0.3}
    }
   }
   n43 = RigidBody(ComponentPose=@n42) {
    bClampLinearTranslationLimitToRefPose=true
    bOverrideWorldGravity=true
    overridePhysicsAsset=&/F_MED_Cavalry_Alt_Physics_tie
    simulationSpace=BaseBoneSpace
    alphaBoolBlend {blendOption=Linear}
    baseBoneRef {boneName=spine_03}
    componentAppliedLinearAccClamp=<10 10 10>
    overrideWorldGravity {z=-980}
   }
   n44 = RigidBody(ComponentPose=@n43) {
    bOverrideWorldGravity=true
    overridePhysicsAsset=&/F_MED_Cavalry_Alt_Physics_skirt
    simulationSpace=BaseBoneSpace
    alphaBoolBlend {blendOption=Linear}
    baseBoneRef {boneName=pelvis}
    componentAppliedLinearAccClamp=<3000 3000 3000>
    componentLinearAccScale=<0.25 0.25 0.25>
    componentLinearVelScale=<0.25 0.25 0.25>
    overrideWorldGravity {z=-980}
    simSpaceSettings {maxAngularAcceleration=500, maxAngularVelocity=3, maxLinearAcceleration=400, maxLinearVelocity=250}
   }
   n45 = RigidBody!(ComponentPose=@n44, bAlphaBoolEnabled:bool=Get bIsFalling(), ExternalForce) -> (Pose) {
    alphaInputType=Bool
    bOverrideWorldGravity=true
    overridePhysicsAsset=&/F_MED_Cavalry_Alt_Physics_skirt_front
    simulationSpace=BaseBoneSpace
    alphaBoolBlend {blendInTime=0.5, blendOption=Linear, blendOutTime=0.5}
    baseBoneRef {boneName=pelvis}
    componentAppliedLinearAccClamp=<3000 3000 3000>
    componentLinearAccScale=<0.25 0.25 0.25>
    componentLinearVelScale=<0.25 0.25 0.25>
    overrideWorldGravity {z=-980}
   }
   n46 = Get dyn_chain_01_base_var()
   n47 = "Transform (Modify) Bone - Bone: dyn_chain_back_01_base_joint"(
    ComponentPose=@n45
    Alpha=@n46
    Translation="0.379,0.985,2.374"
    Rotation=Get!() -> (chain_back_Location:Vector, chain_back_Rotation:Rotator, chain_back_Scale:Vector).chain_back_Rotation
    Scale="1,1,1"
   ) {
    rotationMode=BMM_Additive
    rotationSpace=BCS_BoneSpace
    translationMode=BMM_Additive
    translationSpace=BCS_BoneSpace
    alphaBoolBlend {blendOption=Linear}
    boneToModify {boneName=dyn_chain_back_01_base_joint}
    translation=<0.379 0.9852 2.374>
   }
   n48 = "Transform (Modify) Bone - Bone: dyn_chain_mid_01_joint"(
    ComponentPose=@n47
    Alpha=@n46
    Translation="0,0,0"
    Rotation=Get!() -> (chain_mid_Location:Vector, chain_mid_Rotation:Rotator, chain_mid_Scale:Vector).chain_mid_Rotation
    Scale="1,1,1"
   ) {
    rotationMode=BMM_Additive
    rotationSpace=BCS_BoneSpace
    translationMode=BMM_Additive
    translationSpace=BCS_BoneSpace
    alphaBoolBlend {blendOption=Linear}
    boneToModify {boneName=dyn_chain_mid_01_joint}
    rotation {yaw=8}
   }
   n49 = "Transform (Modify) Bone - Bone: dyn_chain_front_01_base_joint"(
    ComponentPose=@n48
    Alpha=@n46
    Translation="0,0,0"
    Rotation=Get!() -> (chain_front_Location:Vector, chain_front_Rotation:Rotator, chain_front_Scale:Vector).chain_front_Rotation
    Scale="1,1,1"
   ) {
    rotationMode=BMM_Additive
    rotationSpace=BCS_BoneSpace
    translationMode=BMM_Additive
    translationSpace=BCS_BoneSpace
    alphaBoolBlend {blendOption=Linear}
    boneToModify {boneName=dyn_chain_front_01_base_joint}
    rotation=<rot -37.263 14.701 -18.331>
   }
   n50 = Get bIsCrouching()
   n51 = "Transform (Modify) Bone - Bone: dyn_skirt_bk_a_base_r"!(ComponentPose=@n49, bAlphaBoolEnabled:bool=@n50, Translation:Vector="0,0,0", Rotation="0,6,0.000", Scale:Vector="1,1,1") -> (Pose) {
    alphaInputType=Bool
    rotationMode=BMM_Additive
    rotationSpace=BCS_BoneSpace
    translationMode=BMM_Additive
    translationSpace=BCS_BoneSpace
    alphaBoolBlend {blendInTime=0.3, blendOption=Linear, blendOutTime=0.1}
    boneToModify {boneName=dyn_skirt_bk_a_base_r}
    rotation {yaw=6}
   }
   n52 = "Transform (Modify) Bone - Bone: dyn_chain_front_01_base_joint"!(ComponentPose=@n51, bAlphaBoolEnabled:bool=@n50, Translation="0,0,0", Rotation="0,-6,0.000", Scale="1,1,1") -> (Pose) {
    alphaInputType=Bool
    rotationMode=BMM_Additive
    rotationSpace=BCS_BoneSpace
    translationMode=BMM_Additive
    translationSpace=BCS_BoneSpace
    alphaBoolBlend {blendInTime=0.3, blendOption=Linear, blendOutTime=0.3}
    boneToModify {boneName=dyn_chain_front_01_base_joint}
    rotation {yaw=-6}
   }
   n53 = "Transform (Modify) Bone - Bone: dyn_skirt_side_a_base_l"!(ComponentPose=@n52, bAlphaBoolEnabled:bool=Get bIsTargeting(), Translation="0,0,0", Rotation="10,0,0", Scale="1,1,1") -> (Pose) {
    alphaInputType=Bool
    rotationMode=BMM_Additive
    rotationSpace=BCS_BoneSpace
    translationMode=BMM_Additive
    translationSpace=BCS_BoneSpace
    alphaBoolBlend {blendInTime=0.2, blendOption=Linear, blendOutTime=0.2}
    boneToModify {boneName=dyn_skirt_side_a_base_l}
    rotation {pitch=10}
   }
   n54 = "Transform (Modify) Bone - Bone: dyn_skirt_bk_a_base_r"(ComponentPose=@n53, bAlphaBoolEnabled=Get bIsJumping(), Translation="0,0,0", Rotation="-15.977,-3.641,-1.899", Scale="1,1,1") {
    alphaInputType=Bool
    rotationMode=BMM_Additive
    rotationSpace=BCS_BoneSpace
    translationMode=BMM_Additive
    translationSpace=BCS_BoneSpace
    alphaBoolBlend {blendInTime=0.5, blendOption=Linear, blendOutTime=0.5}
    boneToModify {boneName=dyn_skirt_bk_a_base_r}
    rotation=<rot -15.977 -3.641 -1.899>
   }
   n55 = "Transform (Modify) Bone - Bone: dyn_skirt_side_a_base_l"(ComponentPose=@n54, bAlphaBoolEnabled=Get bIsMoving2D(), Translation="0,0,0", Rotation="0.000,-10,0", Scale="1,1,1") {
    alphaInputType=Bool
    rotationMode=BMM_Additive
    rotationSpace=BCS_BoneSpace
    translationMode=BMM_Additive
    translationSpace=BCS_BoneSpace
    alphaBoolBlend {blendInTime=0.2, blendOption=Linear, blendOutTime=0.2}
    boneToModify {boneName=dyn_skirt_side_a_base_l}
    rotation {yaw=-10}
   }
   n56 = "Transform (Modify) Bone - Bone: dyn_skirt_fr_a_base_r"(ComponentPose=@n55, Alpha=1, Translation="0,0,0", Rotation="0,10,0", Scale="1,1,1") {
    rotationMode=BMM_Additive
    rotationSpace=BCS_BoneSpace
    translationMode=BMM_Additive
    translationSpace=BCS_BoneSpace
    alphaBoolBlend {blendInTime=0.2, blendOption=Linear, blendOutTime=0.2}
    boneToModify {boneName=dyn_skirt_fr_a_base_r}
    rotation {yaw=10}
   }
   n57 = "Transform (Modify) Bone - Bone: dyn_skirt_side_a_base_r"!(
    ComponentPose=@n56
    bAlphaBoolEnabled:bool=Get Parachute_OR_Crouching() -> (Parachute_OR_Crouching:bool)
    Translation="0,0,0"
    Rotation="0,5,0"
    Scale="1,1,1"
   ) -> (Pose) {
    alphaInputType=Bool
    rotationMode=BMM_Additive
    rotationSpace=BCS_BoneSpace
    translationMode=BMM_Additive
    translationSpace=BCS_BoneSpace
    alphaBoolBlend {blendInTime=0.2, blendOption=Linear, blendOutTime=0.2}
    boneToModify {boneName=dyn_skirt_side_a_base_r}
    rotation {yaw=5}
   }
   n58 = "Transform (Modify) Bone - Bone: extra_side_thigh_l"!(ComponentPose=@n57, bAlphaBoolEnabled:bool=Get bIsCrouching(), Translation="-1,0,0", Rotation="0,0,0", Scale="1,1,1") -> (Pose) {
    alphaInputType=Bool
    rotationMode=BMM_Additive
    rotationSpace=BCS_BoneSpace
    translationMode=BMM_Additive
    translationSpace=BCS_BoneSpace
    alphaBoolBlend {blendInTime=0.2, blendOption=Linear, blendOutTime=0.2}
    boneToModify {boneName=extra_side_thigh_l}
    translation {x=-1}
   }
   n59 = Get Jumping_OR_Parachute() -> (Jumping_OR_Parachute:bool)
   n60 = "Transform (Modify) Bone - Bone: dyn_skirt_side_a_base_l"(ComponentPose=@n58, bAlphaBoolEnabled=@n59, Translation="0,0,0", Rotation="-5,0,0", Scale="1,1,1") {
    alphaInputType=Bool
    rotationMode=BMM_Additive
    rotationSpace=BCS_BoneSpace
    translationMode=BMM_Additive
    translationSpace=BCS_BoneSpace
    alphaBoolBlend {blendInTime=0.2, blendOption=Linear, blendOutTime=0.2}
    boneToModify {boneName=dyn_skirt_side_a_base_l}
    rotation {pitch=-5}
   }
   n61 = "Transform (Modify) Bone - Bone: extra_side_thigh_l"(ComponentPose=@n60, bAlphaBoolEnabled=@n59, Translation="-1,0,0", Rotation="0,0,0", Scale="1,1,1") {
    alphaInputType=Bool
    rotationMode=BMM_Additive
    rotationSpace=BCS_BoneSpace
    translationMode=BMM_Additive
    translationSpace=BCS_BoneSpace
    alphaBoolBlend {blendInTime=0.2, blendOption=Linear, blendOutTime=0.2}
    boneToModify {boneName=extra_side_thigh_l}
    translation {x=-1}
   }
   n62 = "Transform (Modify) Bone - Bone: dyn_skirt_bk_a_base_l"!(ComponentPose=@n61, bAlphaBoolEnabled:bool=@n59, Translation:Vector="0,0,0", Rotation="0,-3,0", Scale:Vector="1,1,1") -> (Pose) {
    alphaInputType=Bool
    rotationMode=BMM_Additive
    rotationSpace=BCS_BoneSpace
    translationMode=BMM_Additive
    translationSpace=BCS_BoneSpace
    alphaBoolBlend {blendInTime=0.2, blendOption=Linear, blendOutTime=0.2}
    boneToModify {boneName=dyn_skirt_bk_a_base_l}
    rotation {yaw=-3}
   }
   n63 = "Transform (Modify) Bone - Bone: dyn_skirt_bk_a_base"!(ComponentPose=@n62, bAlphaBoolEnabled:bool=@n59, Translation:Vector="0,0,0", Rotation="0,-3,0", Scale:Vector="1,1,1") -> (Pose) {
    alphaInputType=Bool
    rotationMode=BMM_Additive
    rotationSpace=BCS_BoneSpace
    translationMode=BMM_Additive
    translationSpace=BCS_BoneSpace
    alphaBoolBlend {blendInTime=0.2, blendOption=Linear, blendOutTime=0.2}
    boneToModify {boneName=dyn_skirt_bk_a_base}
    rotation {yaw=-3}
   }
   n64 = "Transform (Modify) Bone - Bone: dyn_skirt_bk_a_base_r"(ComponentPose=@n63, bAlphaBoolEnabled=@n59, Translation="0,0,0", Rotation="0,-3,0", Scale="1,1,1") {
    alphaInputType=Bool
    rotationMode=BMM_Additive
    rotationSpace=BCS_BoneSpace
    translationMode=BMM_Additive
    translationSpace=BCS_BoneSpace
    alphaBoolBlend {blendInTime=0.2, blendOption=Linear, blendOutTime=0.2}
    boneToModify {boneName=dyn_skirt_bk_a_base_r}
    rotation {yaw=-3}
   }
   n65 = Get Crouching_AND_Moving2D() -> (Crouching_AND_Moving2D:bool)
   n66 = "Transform (Modify) Bone - Bone: dyn_chain_mid_01_joint"!(ComponentPose=@n64, bAlphaBoolEnabled:bool=@n65, Translation="0,0,0", Rotation="0,0,0", Scale="1,1,1") -> (Pose) {
    alphaInputType=Bool
    rotationMode=BMM_Additive
    rotationSpace=BCS_BoneSpace
    translationMode=BMM_Additive
    translationSpace=BCS_BoneSpace
    alphaBoolBlend {blendInTime=0.2, blendOption=Linear, blendOutTime=0.2}
    boneToModify {boneName=dyn_chain_mid_01_joint}
   }
   n67 = "Transform (Modify) Bone - Bone: dyn_skirt_fr_a_base_r"!(ComponentPose=@n66, bAlphaBoolEnabled:bool=@n65, Translation="0,0,0", Rotation="0,0,0", Scale="1,1,1") -> (Pose) {
    alphaInputType=Bool
    rotationMode=BMM_Additive
    rotationSpace=BCS_BoneSpace
    translationMode=BMM_Additive
    translationSpace=BCS_BoneSpace
    alphaBoolBlend {blendOption=Linear}
    boneToModify {boneName=dyn_skirt_fr_a_base_r}
   }
   n68 = "Transform (Modify) Bone - Bone: dyn_skirt_fr_a_base_l"!(ComponentPose=@n67, bAlphaBoolEnabled:bool=@n65, Translation="0,0,0", Rotation="10,0,0", Scale="1,1,1") -> (Pose) {
    alphaInputType=Bool
    rotationMode=BMM_Additive
    rotationSpace=BCS_BoneSpace
    translationMode=BMM_Additive
    translationSpace=BCS_BoneSpace
    alphaBoolBlend {blendInTime=0.2, blendOption=Linear, blendOutTime=0.2}
    boneToModify {boneName=dyn_skirt_fr_a_base_l}
    rotation {pitch=10}
   }
   n69 = "Transform (Modify) Bone - Bone: dyn_skirt_fr_b_r"!(ComponentPose=@n68, bAlphaBoolEnabled:bool=@n65, Translation="0,0,0", Rotation="0.634,14.690,-4.893", Scale="1,1,1") -> (Pose) {
    alphaInputType=Bool
    rotationMode=BMM_Additive
    rotationSpace=BCS_BoneSpace
    translationMode=BMM_Additive
    translationSpace=BCS_BoneSpace
    alphaBoolBlend {blendInTime=0.2, blendOption=Linear, blendOutTime=0.2}
    boneToModify {boneName=dyn_skirt_fr_b_r}
    rotation=<rot 0.634 14.69 -4.893>
   }
   n70 = "Transform (Modify) Bone - Bone: dyn_skirt_side_b_r"!(ComponentPose=@n69, bAlphaBoolEnabled:bool=@n65, Translation="0,2,0", Rotation="0.000,5,0.000", Scale="1,1,1") -> (Pose) {
    alphaInputType=Bool
    rotationMode=BMM_Additive
    rotationSpace=BCS_BoneSpace
    translationMode=BMM_Additive
    translationSpace=BCS_BoneSpace
    alphaBoolBlend {blendInTime=0.2, blendOption=Linear, blendOutTime=0.2}
    boneToModify {boneName=dyn_skirt_side_b_r}
    rotation {yaw=5}
    translation {y=2}
   }
   n71 = Get static_crouching_rotation() -> (static_crouching_rotation:bool)
   n72 = "Transform (Modify) Bone - Bone: dyn_chain_mid_01_joint"(ComponentPose=@n70, bAlphaBoolEnabled=@n71, Translation="0,0,0", Rotation="0,12,0", Scale="1,1,1") {
    alphaInputType=Bool
    rotationMode=BMM_Additive
    rotationSpace=BCS_BoneSpace
    translationMode=BMM_Additive
    translationSpace=BCS_BoneSpace
    alphaBoolBlend {blendInTime=0.2, blendOption=Linear, blendOutTime=0.2}
    boneToModify {boneName=dyn_chain_mid_01_joint}
    rotation {yaw=12}
   }
   n73 = "Transform (Modify) Bone - Bone: dyn_chain_front_01_base_joint"(ComponentPose=@n72, bAlphaBoolEnabled=@n71, Translation="0,0,0", Rotation="-12,0,0", Scale="1,1,1") {
    alphaInputType=Bool
    rotationMode=BMM_Additive
    rotationSpace=BCS_BoneSpace
    translationMode=BMM_Additive
    translationSpace=BCS_BoneSpace
    alphaBoolBlend {blendInTime=0.2, blendOption=Linear, blendOutTime=0.2}
    boneToModify {boneName=dyn_chain_front_01_base_joint}
    rotation {pitch=-12}
   }
   n74 = Get Left_move() -> (Left_move:bool)
   n75 = "Transform (Modify) Bone - Bone: dyn_chain_front_01_base_joint"(ComponentPose=@n73, bAlphaBoolEnabled=@n74, Translation="0,0,0", Rotation="-12,0,0", Scale="1,1,1") {
    alphaInputType=Bool
    rotationMode=BMM_Additive
    rotationSpace=BCS_BoneSpace
    translationMode=BMM_Additive
    translationSpace=BCS_BoneSpace
    alphaBoolBlend {blendInTime=0.2, blendOption=Linear, blendOutTime=0.2}
    boneToModify {boneName=dyn_chain_front_01_base_joint}
    rotation {pitch=-12}
   }
   n76 = "Transform (Modify) Bone - Bone: dyn_chain_mid_01_joint"(ComponentPose=@n75, bAlphaBoolEnabled=@n74, Translation="0,0,0", Rotation="0,7,0", Scale="1,1,1") {
    alphaInputType=Bool
    rotationMode=BMM_Additive
    rotationSpace=BCS_BoneSpace
    translationMode=BMM_Additive
    translationSpace=BCS_BoneSpace
    alphaBoolBlend {blendInTime=0.2, blendOption=Linear, blendOutTime=0.2}
    boneToModify {boneName=dyn_chain_mid_01_joint}
    rotation {yaw=7}
   }
   n77 = Get bIsSprinting()
   n78 = "Transform (Modify) Bone - Bone: dyn_skirt_bk_a_base_l"(ComponentPose=@n76, bAlphaBoolEnabled=@n77, Translation="0,1,0", Rotation="0,0,0", Scale="1,1,1") {
    alphaInputType=Bool
    rotationMode=BMM_Additive
    rotationSpace=BCS_BoneSpace
    translationMode=BMM_Additive
    translationSpace=BCS_BoneSpace
    alphaBoolBlend {blendInTime=0.2, blendOption=Linear, blendOutTime=0.2}
    boneToModify {boneName=dyn_skirt_bk_a_base_l}
    translation {y=1}
   }
   n79 = "Transform (Modify) Bone - Bone: dyn_skirt_bk_a_base"(ComponentPose=@n78, bAlphaBoolEnabled=@n77, Translation="0,-1,0", Rotation="0,0,0", Scale="1,1,1") {
    alphaInputType=Bool
    rotationMode=BMM_Additive
    rotationSpace=BCS_BoneSpace
    translationMode=BMM_Additive
    translationSpace=BCS_BoneSpace
    alphaBoolBlend {blendInTime=0.2, blendOption=Linear, blendOutTime=0.2}
    boneToModify {boneName=dyn_skirt_bk_a_base}
    translation {y=-1}
   }
   n80 = "Transform (Modify) Bone - Bone: dyn_skirt_bk_a_base_r"(ComponentPose=@n79, bAlphaBoolEnabled=@n77, Translation="0,-1,0", Rotation="0,0,0", Scale="1,1,1") {
    alphaInputType=Bool
    rotationMode=BMM_Additive
    rotationSpace=BCS_BoneSpace
    translationMode=BMM_Additive
    translationSpace=BCS_BoneSpace
    alphaBoolBlend {blendInTime=0.2, blendOption=Linear, blendOutTime=0.2}
    boneToModify {boneName=dyn_skirt_bk_a_base_r}
    translation {y=-1}
   }
   n81 = "Transform (Modify) Bone - Bone: extra_side_thigh_l"!(ComponentPose=@n80, Alpha:float=1, Translation="-2,0,0", Rotation="0,0,0", Scale="1,1,1") -> (Pose) {
    rotationMode=BMM_Additive
    rotationSpace=BCS_BoneSpace
    translationMode=BMM_Additive
    translationSpace=BCS_BoneSpace
    alphaBoolBlend {blendOption=Linear}
    boneToModify {boneName=extra_side_thigh_l}
    translation {x=-2}
   }
   n82 = "Transform (Modify) Bone - Bone: dyn_chain_front_01_base_joint"(
    ComponentPose=@n81
    bAlphaBoolEnabled=Get notIdlePose() -> (notIdlePose:bool)
    Translation="0,0,0"
    Rotation="-2.926,-11.224,-14.343"
    Scale="1,1,1"
   ) {
    alphaInputType=Bool
    rotationMode=BMM_Additive
    rotationSpace=BCS_BoneSpace
    translationMode=BMM_Additive
    translationSpace=BCS_BoneSpace
    alphaBoolBlend {blendInTime=0.1, blendOption=Linear, blendOutTime=0.2}
    boneToModify {boneName=dyn_chain_front_01_base_joint}
    rotation=<rot -2.926 -11.224 -14.343>
   }
   n83 = "Transform (Modify) Bone - Bone: dyn_chain_front_01_base_joint"(
    ComponentPose=@n82
    bAlphaBoolEnabled=NOT Boolean(A=Get notIdlePose())
    Translation="0.043,-0.160,0.750"
    Rotation="-4.195,-4.5,-13.690"
    Scale="1,1,1"
   ) {
    alphaInputType=Bool
    rotationMode=BMM_Additive
    rotationSpace=BCS_BoneSpace
    translationMode=BMM_Additive
    translationSpace=BCS_BoneSpace
    alphaBoolBlend {blendInTime=0.1, blendOption=Linear, blendOutTime=0.2}
    boneToModify {boneName=dyn_chain_front_01_base_joint}
    rotation=<rot -4.195 -4.5 -13.69>
    translation=<0.043 -0.16 0.75>
   }
   n84 = "Transform (Modify) Bone - Bone: dyn_skirt_side_a_base_r"!(ComponentPose=@n83, Alpha:float=1, Translation="0,0.5,0", Rotation="0,0,0", Scale="1,1,1") -> (Pose) {
    rotationMode=BMM_Additive
    rotationSpace=BCS_BoneSpace
    translationMode=BMM_Additive
    translationSpace=BCS_BoneSpace
    alphaBoolBlend {blendOption=Linear}
    boneToModify {boneName=dyn_skirt_side_a_base_r}
    translation {y=0.5}
   }
   n85 = "Transform (Modify) Bone - Bone: dyn_chain_mid_01_joint"(ComponentPose=@n84, bAlphaBoolEnabled=NOT Boolean(A=Get notIdlePose()), Translation="0,0,0", Rotation="-0.004,-0.011,0.000", Scale="1,1,1") {
    alphaInputType=Bool
    rotationMode=BMM_Additive
    rotationSpace=BCS_BoneSpace
    translationMode=BMM_Additive
    translationSpace=BCS_BoneSpace
    alphaBoolBlend {blendInTime=0.1, blendOption=Linear, blendOutTime=0.2}
    boneToModify {boneName=dyn_chain_mid_01_joint}
    rotation=<rot -0.004 -0.011 0>
   }
   n86 = "Transform (Modify) Bone - Bone: dyn_chain_mid_01_joint"(ComponentPose=@n85, bAlphaBoolEnabled=Get notIdlePose(), Translation="0,0.5,0", Rotation="0,3,0", Scale="1,1,1") {
    alphaInputType=Bool
    rotationMode=BMM_Additive
    rotationSpace=BCS_BoneSpace
    translationMode=BMM_Additive
    translationSpace=BCS_BoneSpace
    alphaBoolBlend {blendInTime=0.1, blendOption=Linear, blendOutTime=0.2}
    boneToModify {boneName=dyn_chain_mid_01_joint}
    rotation {yaw=3}
    translation {y=0.5}
   }
   n87 = "Transform (Modify) Bone - Bone: dyn_chain_back_01_base_joint"!(ComponentPose=@n86, bAlphaBoolEnabled:bool=NOT Boolean(A=Get notIdlePose()), Translation="0,0,0", Rotation="0,20.002,0", Scale="1,1,1") -> (Pose) {
    alphaInputType=Bool
    rotationMode=BMM_Additive
    rotationSpace=BCS_BoneSpace
    translationMode=BMM_Additive
    translationSpace=BCS_BoneSpace
    alphaBoolBlend {blendInTime=0.1, blendOption=Linear, blendOutTime=0.2}
    boneToModify {boneName=dyn_chain_back_01_base_joint}
    rotation {yaw=20.002}
   }
   n88 = Flap Node - Bone: dyn_bracelet_bottom_2_base!(ComponentPose:Component Space Pose Link Structure=@n87, Alpha:float=0.5) -> (Pose:Component Space Pose Link Structure) {
    is=title FortAnimGraphNode_Flap
    bUseClamp=true
    targetClampMin=-3.14
    targetComponent=RotationZ
    alphaBoolBlend {blendOption=Linear}
    sourceDrives [{
      bUseRange=false
      drivingCurve=null
      multiplier=-0.75
      rangeMax=0
      rangeMin=0
      remappedMax=0
      remappedMin=0
      sourceComponent=RotationY
      useQuaternion=true
      sourceBone {boneName=hand_r}
     }]
    targetBone {boneName=dyn_bracelet_bottom_2_base}
   }
   Output Pose(Result=Component To Local(ComponentPose=@n88)) {layerGroup=None}
  ]}
 EventGraph {nodes [
   Event BlueprintUpdateAnimation()
   > n1 = Sequence!(execute) -> (then_0, then_1, then_2:Exec, then_3:Exec, then_4:Exec, then_5:Exec, then_6:Exec, then_7:Exec)
   n2 = Get bIsMoving2D()
   n3 = "Absolute (Float)"(A=BreakRotator(InRot=Get PawnDeltaRotation() -> (PawnDeltaRotation:Rotator)).Yaw)
   > Set static_crouching_rotation(static_crouching_rotation:bool=AND Boolean!(A="float > float"(A=@n3, B=2), B=Get bIsCrouching(), C:bool=NOT Boolean(A=@n2)) -> (ReturnValue)) -> (Output_Get:bool)
   n4 = GetOwnerSkeletalMeshComponent()
   Set skirt_fr_a_base_l_var(
    execute=@n1.then_1
    skirt_fr_a_base_l_var:double=MapRangeClamped(Value=GetTwistAndSwingAngleOfDeltaRotationFromRefPose(self=@n4, BoneName=thigh_l).OutSwingAngle, InRangeA=23, InRangeB=115)
   ) -> (Output_Get:double)
   n5 = MapRangeClamped(Value=GetTwistAndSwingAngleOfDeltaRotationFromRefPose(self=@n4, BoneName=thigh_r).OutSwingAngle, InRangeA=172, InRangeB=70)
   > Set dyn_chain_01_base_var(dyn_chain_01_base_var:double=@n5) -> (Output_Get:double)
   > Set skirt_fr_a_base_r_var(skirt_fr_a_base_r_var:double=@n5) -> (Output_Get:double)
   n6 = Get bIsCrouching()
   > n7 = Branch(Condition=AND Boolean!(A=Get bIsSprinting(), B=NOT Boolean(A=@n6)) -> (ReturnValue))
   > Set L_swinging_back_skirt(L_swinging_back_skirt:double=Get skirt_fr_a_base_l_var()) -> (Output_Get:double)
   > Set R_swinging_back_skirt(R_swinging_back_skirt:double=Get skirt_fr_a_base_r_var()) -> (Output_Get:double)
   Set L_swinging_back_skirt(execute=@n7.else, L_swinging_back_skirt=0)
   > Set R_swinging_back_skirt(R_swinging_back_skirt=0)
   n8 = Branch(execute=@n1.then_2, Condition=AND Boolean(A=Get bIsCrouching(), B=Get bIsMoving2D()))
   n9 = Enum to String!(Enumerator:EFortCardinalDirection Enum=GetCardinalDirectionFromAngle!(
     self:Fort Anim Instance Object Reference
     Angle:float=Get PawnMovementDirectionAzimuth() -> (PawnMovementDirectionAzimuth:float)
     DeadZoneAngle:float=0
    ) -> (ReturnValue:EFortCardinalDirection Enum=North)) -> (ReturnValue:String)
   > Set Right_move(Right_move:bool="Equal Exactly (String)"(A=@n9, B=East)) -> (Output_Get:bool)
   > Set Left_move(Left_move:bool="Equal Exactly (String)"(A=@n9, B=West)) -> (Output_Get:bool)
   Set Right_move(execute=@n8.else, Right_move=false)
   > Set Left_move(Left_move=false)
   n10 = Branch(execute=@n1.then_3, Condition=Get bIsCrouching())
   > Set!(execute, chain_mid_Location:Vector="0, 0, 0", chain_mid_Rotation:Rotator="0,8,0", chain_mid_Scale:Vector="1,1,1") -> (then, Output_Get:Transform)
   > Set!(execute, chain_back_Location:Vector="0,0,20", chain_back_Rotation:Rotator="0, 0, 0", chain_back_Scale:Vector="1,1,1") -> (then, Output_Get)
   > Set!(execute, chain_front_Location:Vector="0, 0, 0", chain_front_Rotation:Rotator="0, 0, 0", chain_front_Scale:Vector="1,1,1") -> (then, Output_Get)
   > Set!(
    execute
    dyn_skirt_fr_a_base_r_Var_Location:Vector="-1.555,3.538,2.405"
    dyn_skirt_fr_a_base_r_Var_Rotation:Rotator="2.705,10.75,-9.391"
    dyn_skirt_fr_a_base_r_Var_Scale:Vector="1,1,1"
   ) -> (then, Output_Get)
   > Set!(
    execute
    dyn_skirt_side_a_base_r_var_Location:Vector="0, 0, 0"
    dyn_skirt_side_a_base_r_var_Rotation:Rotator="0, 0, 0"
    dyn_skirt_side_a_base_r_var_Scale:Vector="1,1,1"
   ) -> (then, Output_Get)
   > Branch(Condition=Get bIsMoving2D())
   > Set!(execute, chain_back_Location:Vector="0, 0, 0", chain_back_Rotation:Rotator="-12.222,17.241,34.297", chain_back_Scale:Vector="1,1,1") -> (then, Output_Get)
   > Branch(Condition=Get Right_move() -> (Right_move:bool))
   > Set!(execute, chain_mid_Location:Vector="0, 0, 0", chain_mid_Rotation:Rotator="0,30,0", chain_mid_Scale:Vector="1,1,1") -> (then, Output_Get)
   > Set!(execute, chain_back_Location:Vector="0, 0, 0", chain_back_Rotation:Rotator="0, 0, 0", chain_back_Scale:Vector="1,1,1") -> (then, Output_Get)
   > Set!(execute, chain_front_Location:Vector="0, 0, 0", chain_front_Rotation:Rotator="0, 0, 0", chain_front_Scale:Vector="1,1,1") -> (then, Output_Get)
   Set!(execute=@n10.else, chain_mid_Location:Vector="0, 0, 0", chain_mid_Rotation:Rotator="0,10.7,0", chain_mid_Scale:Vector="1,1,1") -> (then, Output_Get)
   > Set!(execute, chain_back_Location:Vector="0, 0, 0", chain_back_Rotation:Rotator="-24.11,75.443,-45.717", chain_back_Scale:Vector="1,1,1") -> (then, Output_Get)
   > Set!(execute, chain_front_Location:Vector="0, 0, 0", chain_front_Rotation:Rotator="0,-28.198, 0", chain_front_Scale:Vector="1,1,1") -> (then, Output_Get)
   > Set!(
    execute
    dyn_skirt_side_a_base_r_var_Location:Vector="0, 0, 0"
    dyn_skirt_side_a_base_r_var_Rotation:Rotator="23.5,16,7"
    dyn_skirt_side_a_base_r_var_Scale:Vector="1,1,1"
   ) -> (then, Output_Get)
   > Set!(
    execute
    dyn_skirt_fr_a_base_r_Var_Location:Vector="0,3, 0"
    dyn_skirt_fr_a_base_r_Var_Rotation:Rotator="0,27, 0"
    dyn_skirt_fr_a_base_r_Var_Scale:Vector="1,1,1"
   ) -> (then, Output_Get)
   n11 = Get bIsParachuteOpened()
   n12 = Get bIsSkyDiving()
   n13 = Get bIsJumping()
   n14 = Get bIsFalling()
   Set R_thigh volume(
    execute=@n1.then_4
    R_thigh volume:bool=AND Boolean!(A=OR Boolean(A=@n14, B=@n13), B=NOT Boolean(A=@n11), C:bool=NOT Boolean(A=@n12)) -> (ReturnValue)
   ) -> (Output_Get:bool)
   > Set Jumping_OR_Parachute(Jumping_OR_Parachute:bool=OR Boolean(A=Get bIsJumping(), B=Get bIsParachuteOpened())) -> (Output_Get:bool)
   > Set Parachute_OR_Crouching(Parachute_OR_Crouching:bool=OR Boolean(A=Get bIsParachuteOpened(), B=Get bIsCrouching())) -> (Output_Get:bool)
   > Set Crouching_AND_Moving2D(Crouching_AND_Moving2D:bool=AND Boolean!(A=Get bIsCrouching(), B=Get bIsMoving2D()) -> (ReturnValue)) -> (Output_Get:bool)
   > Set notIdlePose(notIdlePose:bool=OR Boolean!(A=Get bIsMoving2D(), B=Get bIsJumping(), C:bool=Get bIsCrouching(), D:bool=Get bIsParachuteOpened(), E:bool=Get bIsPlayingEmote(), F:bool) -> (ReturnValue)) -> (Output_Get:bool)
   > Set LOD_2_3(LOD_2_3:bool="integer >= integer"(A=Get CurrentLOD(), B=1)) -> (Output_Get:bool)
   n15 = GetOwnerAnimationInterface!(self:Custom Character Part Anim Instance Object Reference) -> (ReturnValue:Fort Player Animation Interface Object Reference)
   > n16 = Branch(Condition="Equal (Enum)"(
     A:EFortWeaponCoreAnimation Enum=GetWeaponCoreAnim!(self:Fort Player Animation Interface Object Reference=@n15) -> (ReturnValue:EFortWeaponCoreAnimation Enum)
     B:EFortWeaponCoreAnimation Enum=Pistol
    ))
   n17 = GetTwistAndSwingAngleOfDeltaRotationFromRefPose(self=GetOwnerSkeletalMeshComponent(), BoneName=upperarm_r)
   > Set L_aim_var_float(L_aim_var_float:double=MapRangeClamped(Value=@n17.OutSwingAngle, InRangeA=110, InRangeB=135)) -> (Output_Get:double)
   Set L_aim_var_float(execute=@n16.else, L_aim_var_float=MapRangeClamped(Value=@n17.OutSwingAngle, InRangeA=80, InRangeB=88))
   Branch(execute=@n1.then_5, Condition=NOT Boolean(A=Get bIsParachuteOpened()))
   n18 = GetTwistAndSwingAngleOfDeltaRotationFromRefPose(self=GetOwnerSkeletalMeshComponent(), BoneName=upperarm_r)
   > Set R_aim_var_bool(R_aim_var_bool:bool="float > float"(A=@n18, B=40)) -> (Output_Get:bool)
   > Set upperarm_r_var(upperarm_r_var:double=MapRangeClamped(Value=@n18.OutSwingAngle, InRangeA=85, InRangeB=108, OutRangeB=0.4)) -> (Output_Get:double)
   > Set collar_r_bool_var(collar_r_bool_var:bool=OR Boolean!(A=Get bIsSprinting(), B=Get bIsCrouching(), C=Get R_aim_var_bool()) -> (ReturnValue)) -> (Output_Get:bool)
   > n19 = Branch(Condition=Get bIsCrouching())
   > Set scale_collar_r(scale_collar_r:double=-1.05) -> (Output_Get:double)
   Set scale_collar_r(execute=@n19.else, scale_collar_r=-0.8)
   GetOwnerDirection!(self:Custom Character Part Anim Instance Object Reference) -> (ReturnValue:float)
  ]}
]
variables [
 chain_back {type=Transform, default {location=<0 0 0>, rotation=<rot 0 0 0>, scale=<1 1 1>}}
 chain_front {type=Transform, default {location=<0 0 0>, rotation=<rot 0 0 0>, scale=<1 1 1>}}
 chain_mid {type=Transform, default {location=<0 0 0>, rotation=<rot 0 0 0>, scale=<1 1 1>}}
 collar_r_bool_var {type=boolean}
 Crouching_AND_Moving2D {type=boolean}
 dyn_chain_01_base_var {type=number}
 dyn_skirt_fr_a_base_r_Var {type=Transform, default {location=<0 0 0>, rotation=<rot 0 0 0>, scale=<1 1 1>}}
 dyn_skirt_side_a_base_r_var {type=Transform, default {location=<0 0 0>, rotation=<rot 0 0 0>, scale=<1 1 1>}}
 Jumping_OR_Parachute {type=boolean}
 L_aim_var_bool {type=boolean}
 L_aim_var_float {type=number}
 L_swinging_back_skirt {type=number}
 Left_move {type=boolean}
 LOD_2_3 {type=boolean}
 notIdlePose {type=boolean}
 Parachute_OR_Crouching {type=boolean}
 R_aim_var_bool {type=boolean}
 R_swinging_back_skirt {type=number}
 R_thigh volume {type=boolean}
 Right_move {type=boolean}
 scale_collar_r {type=number}
 skirt_fr_a_base_l_var {type=number}
 skirt_fr_a_base_r_var {type=number}
 static_crouching_rotation {type=boolean}
 upperarm_r_var {type=number}
]
