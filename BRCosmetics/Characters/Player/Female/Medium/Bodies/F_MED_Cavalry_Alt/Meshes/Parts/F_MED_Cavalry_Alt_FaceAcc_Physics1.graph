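# Physics asset for the face accessory mesh of the female medium Cavalry Alt character, setting up physics for its dynamic hair bones.
# Five three-bone capsule chains (center, left, right, back-left, back-right) hang from the head through constraints with narrow swing and twist limits.
# Larger capsules on head, neck_01 and spine_05 supply body volumes, with no constraints linking these three bones to one another.
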PhysicsAsset /BRCosmetics/Characters/Player/Female/Medium/Bodies/F_MED_Cavalry_Alt/Meshes/Parts/F_MED_Cavalry_Alt_FaceAcc_Physics1
bodies {
 dyn_hair_01 [{shapeType=Capsule, length=4.1562, radius=1.125, shapeName=dyn_hair_01_capsule, center=<3.5478 0 0>, rotation=<rot -90 160.5288 -160.5288>}]
 dyn_hair_01_l [{shapeType=Capsule, length=3.0312, radius=1.3438, shapeName=dyn_hair_01_l_capsule, center=<3.1485 0 0>, rotation=<rot -90 180 -179.9996>}]
 dyn_hair_01_r [{shapeType=Capsule, length=4.5625, radius=1.1562, shapeName=dyn_hair_01_r_capsule, center=<-3.5401 0 0>, rotation=<rot -90 90 -89.9996>}]
 dyn_hair_02 [{shapeType=Capsule, length=5.5625, radius=1.75, shapeName=dyn_hair_02_capsule, center=<2.1598 0 0>, rotation=<rot -90 177.6438 -89.798>}]
 dyn_hair_02_l [{shapeType=Capsule, length=2.7188, radius=1.5938, shapeName=dyn_hair_02_l_capsule, center=<3.288 0 0>, rotation=<rot -90 -90 90>}]
 dyn_hair_02_r [{
   shapeType=Capsule
   length=1.6875
   radius=1.6562
   shapeName=dyn_hair_02_r_capsule
   center=<-2.2502 0 -0.0217>
   rotation=<rot 90 88.0715 86.1974>
  }]
 dyn_hair_03 [{shapeType=Capsule, length=5.4375, radius=1.7812, shapeName=dyn_hair_03_capsule, center=<1.8721 0 -0.0206>, rotation=<rot 90 -2.3562 89.1816>}]
 dyn_hair_03_l [{
   shapeType=Capsule
   length=1.5938
   radius=1.5938
   shapeName=dyn_hair_03_l_capsule
   center=<2.5512 0 0>
   rotation=<rot -90 -179.9999 -179.9999>
  }]
 dyn_hair_03_r [{
   shapeType=Capsule
   length=5.7188
   radius=1.0938
   shapeName=dyn_hair_03_r_capsule
   center=<-2.0885 0 0.1662>
   rotation=<rot 90 88.0715 -1.6804>
  }]
 dyn_hair_back_01_l [{
   shapeType=Capsule
   length=4.0312
   radius=1.3438
   shapeName=dyn_hair_back_01_l_capsule
   center=<3.2272 0 0.2823>
   rotation=<rot -90 176.4805 -176.4801>
  }]
 dyn_hair_back_01_r [{shapeType=Capsule, length=3, radius=1.625, shapeName=dyn_hair_back_01_r_capsule, center=<-4.4929 -0.0005 0>, rotation=<rot 90 180 89.9791>}]
 dyn_hair_back_02_l [{shapeType=Capsule, length=1.875, radius=1.3438, shapeName=dyn_hair_back_02_l_capsule, center=<2.5164 0 0.2202>, rotation=<rot -85 0.0003 0>}]
 dyn_hair_back_02_r [{
   shapeType=Capsule
   length=5.0563
   radius=1.6187
   shapeName=dyn_hair_back_02_r_capsule
   center=<-2.8406 0.0227 -0.2485>
   rotation=<rot -85.0001 0.0002 88.3037>
  }]
 dyn_hair_back_03_l [{
   shapeType=Capsule
   length=1.2188
   radius=2.0312
   shapeName=dyn_hair_back_03_l_capsule
   center=<2.1519 0 0.3118>
   rotation=<rot 90 0.9076 0.2147>
  }]
 dyn_hair_back_03_r [{
   shapeType=Capsule
   length=4.6875
   radius=2
   shapeName=dyn_hair_back_03_r_capsule
   center=<-2.9492 0.0377 -0.258>
   rotation=<rot 85.0001 -179.9997 92.1864>
  }]
 head [
  {shapeType=Capsule, length=6.7534, radius=6.3369, center=<3.1734 2.6019 0.0541>, rotation=<rot 0.5016 -90.415 -88.3249>}
  {
   shapeType=Capsule
   length=3.3889
   radius=6.0811
   shapeName=head_capsule
   center=<1.6133 -0.0375 -0.0239>
   rotation=<rot -89.1519 179.9981 -180>
  }
 ]
 neck_01 [{
   shapeType=Capsule
   length=6.0949
   radius=5.2884
   shapeName=neck_01_capsule
   center=<1.5312 0.874 -0.0664>
   rotation=<rot 87.5172 -0.0032 -11.1613>
  }]
 spine_05 [
  {shapeType=Capsule, length=8.0453, radius=4.3542, shapeName=spine_05_capsule, center=<8.3727 -1.9414 0>}
  {
   shapeType=Capsule
   length=4.3766
   radius=10.8207
   shapeName=spine_05_capsule
   center=<0.9477 1.8096 -0.0043>
   rotation=<rot -89.5488 -69.2055 73.8815>
  }
 ]
}
constraints [
 {bone1Name=dyn_hair_02, bone2Name=dyn_hair_01, swing1LimitDegrees=10, swing2LimitDegrees=20, twistLimitDegrees=1}
 {bone1Name=dyn_hair_03, bone2Name=dyn_hair_02, swing1LimitDegrees=10, swing2LimitDegrees=20, twistLimitDegrees=1}
 {bone1Name=dyn_hair_01, bone2Name=head, swing1LimitDegrees=10, swing2LimitDegrees=30, twistLimitDegrees=1}
 {bone1Name=dyn_hair_01_l, bone2Name=head, swing1LimitDegrees=15, swing2LimitDegrees=15, twistLimitDegrees=1}
 {bone1Name=dyn_hair_02_l, bone2Name=dyn_hair_01_l, swing1LimitDegrees=10, swing2LimitDegrees=15, twistLimitDegrees=1}
 {bone1Name=dyn_hair_03_l, bone2Name=dyn_hair_02_l, swing1LimitDegrees=10, swing2LimitDegrees=15, twistLimitDegrees=1}
 {bone1Name=dyn_hair_back_01_l, bone2Name=head, swing1LimitDegrees=10, swing2LimitDegrees=20, twistLimitDegrees=1}
 {bone1Name=dyn_hair_back_02_l, bone2Name=dyn_hair_back_01_l, swing1LimitDegrees=10, swing2LimitDegrees=20, twistLimitDegrees=1}
 {bone1Name=dyn_hair_back_03_l, bone2Name=dyn_hair_back_02_l, swing1LimitDegrees=10, swing2LimitDegrees=20, twistLimitDegrees=1}
 {bone1Name=dyn_hair_back_01_r, bone2Name=head, swing1LimitDegrees=10, swing2LimitDegrees=20, twistLimitDegrees=1}
 {bone1Name=dyn_hair_back_02_r, bone2Name=dyn_hair_back_01_r, swing1LimitDegrees=10, swing2LimitDegrees=20, twistLimitDegrees=1}
 {bone1Name=dyn_hair_back_03_r, bone2Name=dyn_hair_back_02_r, swing1LimitDegrees=10, swing2LimitDegrees=20, twistLimitDegrees=1}
 {bone1Name=dyn_hair_01_r, bone2Name=head, swing1LimitDegrees=15, swing2LimitDegrees=15, twistLimitDegrees=1}
 {bone1Name=dyn_hair_02_r, bone2Name=dyn_hair_01_r, swing1LimitDegrees=10, swing2LimitDegrees=15, twistLimitDegrees=5}
 {bone1Name=dyn_hair_03_r, bone2Name=dyn_hair_02_r, swing1LimitDegrees=10, swing2LimitDegrees=15, twistLimitDegrees=5}
]
