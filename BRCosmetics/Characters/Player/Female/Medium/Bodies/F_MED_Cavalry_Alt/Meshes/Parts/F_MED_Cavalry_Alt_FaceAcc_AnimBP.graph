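# Post-physics AnimBlueprint for the female Cavalry Alt face-accessory part, which simulates its hair and applies pose-specific corrections.
# The parent mesh pose feeds a head-space RigidBody node with per-update gravity override, then a cached pose and additive hair-bone transforms.
# For pistol and several long-gun animations, targeting, crouch, movement and relaxed states choose which of six `head_up_massive` alphas follows head-to-spine_05 rotation.
# An unconnected transform chain remains in the AnimGraph.
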AnimBlueprint /BRCosmetics/Characters/Player/Female/Medium/Bodies/F_MED_Cavalry_Alt/Meshes/Parts/F_MED_Cavalry_Alt_FaceAcc_AnimBP
target_skeleton=&/F_MED_Cavalry_Alt_FaceAcc_Skeleton
graphs [
 AnimGraph {nodes [
   n1 = "Get (a copy)"(Array:"Array of Float (double-precision)s"=Get head_up_massive() -> (head_up_massive:"Array of Float (double-precision)s")) -> (Output:double)
   n2 = "Transform (Modify) Bone - Bone: dyn_hair_back_01_r"(Alpha=@n1, Translation="0,0,0", Rotation="0,0,5", Scale="1,1,1") {
    rotationMode=BMM_Additive
    translationMode=BMM_Additive
    alphaBoolBlend {blendOption=Linear}
    alphaScaleBiasClamp {bInterpResult=true}
    boneToModify {boneName=dyn_hair_back_01_r}
    rotation {roll=5}
   }
   n3 = "Transform (Modify) Bone - Bone: dyn_hair_01_r"(ComponentPose=@n2, Alpha=@n1, Translation="0,0,0", Rotation="-10,-0.000,2", Scale="1,1,1") {
    rotationMode=BMM_Additive
    translationMode=BMM_Additive
    alphaBoolBlend {blendOption=Linear}
    alphaScaleBiasClamp {bInterpResult=true}
    boneToModify {boneName=dyn_hair_01_r}
    rotation {pitch=-10, roll=2}
   }
   n4 = "Transform (Modify) Bone - Bone: dyn_hair_back_02_r"(ComponentPose=@n3, Alpha=@n1, Translation="0,0,0", Rotation="-10,5,-10", Scale="1,1,1") {
    rotationMode=BMM_Additive
    translationMode=BMM_Additive
    alphaBoolBlend {blendOption=Linear}
    alphaScaleBiasClamp {bInterpResult=true}
    boneToModify {boneName=dyn_hair_back_02_r}
    rotation=<rot -10 5 -10>
   }
   n5 = "Transform (Modify) Bone - Bone: dyn_hair_02_r"(ComponentPose=@n4, Alpha=@n1, Translation="0,0,0", Rotation="0,0,0", Scale="1,1,1") {rotationMode=BMM_Additive, translationMode=BMM_Additive, alphaBoolBlend {blendOption=Linear}, boneToModify {boneName=dyn_hair_02_r}}
   n6 = "Transform (Modify) Bone - Bone: dyn_hair_back_03_r"(ComponentPose=@n5, Alpha=@n1, Translation="0,0,0", Rotation="0,0,0", Scale="1,1,1") {rotationMode=BMM_Additive, translationMode=BMM_Additive, alphaBoolBlend {blendOption=Linear}, boneToModify {boneName=dyn_hair_back_03_r}}
   n7 = "Transform (Modify) Bone - Bone: dyn_hair_03_r"(ComponentPose=@n6, Alpha=@n1, Translation="0,0,0", Rotation="0,0,0", Scale="1,1,1") {rotationMode=BMM_Additive, translationMode=BMM_Additive, alphaBoolBlend {blendOption=Linear}, boneToModify {boneName=dyn_hair_03_r}}
   n8 = "Get (a copy)"(Array=Get head_up_massive(), Dimension 1=1)
   n9 = "Transform (Modify) Bone - Bone: dyn_hair_back_01_r"(ComponentPose=@n7, Alpha=@n8, Translation="0,0,0", Rotation="-6.071,0.002,26.586", Scale="1,1,1") {
    rotationMode=BMM_Additive
    translationMode=BMM_Additive
    alphaBoolBlend {blendOption=Linear}
    alphaScaleBiasClamp {bInterpResult=true}
    boneToModify {boneName=dyn_hair_back_01_r}
    rotation=<rot -6.071 0.002 26.586>
   }
   n10 = "Transform (Modify) Bone - Bone: dyn_hair_01_r"(ComponentPose=@n9, Alpha=@n8, Translation="0,0,0", Rotation="-11.667,-2.107,16.483", Scale="1,1,1") {
    rotationMode=BMM_Additive
    translationMode=BMM_Additive
    alphaBoolBlend {blendOption=Linear}
    alphaScaleBiasClamp {bInterpResult=true}
    boneToModify {boneName=dyn_hair_01_r}
    rotation=<rot -11.667 -2.107 16.483>
   }
   n11 = "Transform (Modify) Bone - Bone: dyn_hair_back_02_r"(ComponentPose=@n10, Alpha=@n8, Translation="0,0,0", Rotation="0,0,0", Scale="1,1,1") {
    rotationMode=BMM_Additive
    translationMode=BMM_Additive
    alphaBoolBlend {blendOption=Linear}
    alphaScaleBiasClamp {bInterpResult=true}
    boneToModify {boneName=dyn_hair_back_02_r}
   }
   n12 = "Transform (Modify) Bone - Bone: dyn_hair_02_r"(ComponentPose=@n11, Alpha=@n8, Translation="0,0,0", Rotation="0,0,0", Scale="1,1,1") {
    rotationMode=BMM_Additive
    translationMode=BMM_Additive
    alphaBoolBlend {blendOption=Linear}
    alphaScaleBiasClamp {bInterpResult=true}
    boneToModify {boneName=dyn_hair_02_r}
   }
   n13 = "Transform (Modify) Bone - Bone: dyn_hair_back_03_r"(ComponentPose=@n12, Alpha=@n8, Translation="0,0,0", Rotation="0,0,0", Scale="1,1,1") {
    rotationMode=BMM_Additive
    translationMode=BMM_Additive
    alphaBoolBlend {blendOption=Linear}
    alphaScaleBiasClamp {bInterpResult=true}
    boneToModify {boneName=dyn_hair_back_03_r}
   }
   "Transform (Modify) Bone - Bone: dyn_hair_03_r"(ComponentPose=@n13, Alpha=@n8, Translation="0,0,0", Rotation="0,0,0", Scale="1,1,1") {
    rotationMode=BMM_Additive
    translationMode=BMM_Additive
    alphaBoolBlend {blendOption=Linear}
    alphaScaleBiasClamp {bInterpResult=true}
    boneToModify {boneName=dyn_hair_03_r}
   }
   n14 = "Get (a copy)"(Array=Get head_up_massive())
   n15 = "Transform (Modify) Bone - Bone: dyn_hair_back_01_r"(
    ComponentPose=Local To Component(LocalPose=Use cached pose 'SavedPose15620'())
    Alpha=@n14
    Translation="-0.000,-0.720,1.549"
    Rotation="-4.836,0.264,-0.362"
    Scale="1,1,1"
   ) {
    rotationMode=BMM_Additive
    translationMode=BMM_Additive
    alphaBoolBlend {blendOption=Linear}
    alphaScaleBiasClamp {bInterpResult=true}
    boneToModify {boneName=dyn_hair_back_01_r}
    rotation=<rot -4.836 0.264 -0.362>
    translation {y=-0.72, z=1.549}
   }
   n16 = "Transform (Modify) Bone - Bone: dyn_hair_01_r"(ComponentPose=@n15, Alpha=@n14, Translation="0,0,0", Rotation="3.298,-3.485,7.223", Scale="1,1,1") {
    rotationMode=BMM_Additive
    translationMode=BMM_Additive
    alphaBoolBlend {blendOption=Linear}
    alphaScaleBiasClamp {bInterpResult=true}
    boneToModify {boneName=dyn_hair_01_r}
    rotation=<rot 3.298 -3.485 7.223>
   }
   n17 = "Transform (Modify) Bone - Bone: dyn_hair_back_02_r"(ComponentPose=@n16, Alpha=@n14, Translation="-0.401,-0.663,2.312", Rotation="3.361,-10.553,-15.399", Scale="1,1,1") {
    rotationMode=BMM_Additive
    translationMode=BMM_Additive
    alphaBoolBlend {blendOption=Linear}
    alphaScaleBiasClamp {bInterpResult=true}
    boneToModify {boneName=dyn_hair_back_02_r}
    rotation=<rot 3.361 -10.553 -15.399>
    translation=<-0.401 -0.663 2.312>
   }
   n18 = "Transform (Modify) Bone - Bone: dyn_hair_02_r"(ComponentPose=@n17, Alpha=@n14, Translation="-1.601,-1.626,2.893", Rotation="6.107,-19.236,-17.976", Scale="1,1,1") {
    rotationMode=BMM_Additive
    translationMode=BMM_Additive
    alphaBoolBlend {blendOption=Linear}
    alphaScaleBiasClamp {bInterpResult=true}
    boneToModify {boneName=dyn_hair_02_r}
    rotation=<rot 6.107 -19.236 -17.976>
    translation=<-1.6009 -1.626 2.893>
   }
   n19 = "Transform (Modify) Bone - Bone: dyn_hair_back_03_r"(ComponentPose=@n18, Alpha=@n14, Translation="-1.002,-1.709,3.986", Rotation="1.406,-0.872,-13.094", Scale="1,1,1") {
    rotationMode=BMM_Additive
    translationMode=BMM_Additive
    alphaBoolBlend {blendOption=Linear}
    alphaScaleBiasClamp {bInterpResult=true}
    boneToModify {boneName=dyn_hair_back_03_r}
    rotation=<rot 1.406 -0.872 -13.094>
    translation=<-1.002 -1.709 3.986>
   }
   n20 = "Transform (Modify) Bone - Bone: dyn_hair_03_r"(ComponentPose=@n19, Alpha=@n14, Translation="-0.000,-0.000,1", Rotation="10,0.000,0.000", Scale="1,1,1") {
    rotationMode=BMM_Additive
    translationMode=BMM_Additive
    alphaBoolBlend {blendOption=Linear}
    alphaScaleBiasClamp {bInterpResult=true}
    boneToModify {boneName=dyn_hair_03_r}
    rotation {pitch=10}
    translation {z=1}
   }
   n21 = "Get (a copy)"(Array=Get head_up_massive(), Dimension 1=1)
   n22 = "Transform (Modify) Bone - Bone: dyn_hair_back_01_r"(ComponentPose=@n20, Alpha=@n21, Translation="-0.000,0.000,0.944", Rotation="0.033,0.044,-5.293", Scale="1,1,1") {
    rotationMode=BMM_Additive
    translationMode=BMM_Additive
    alphaBoolBlend {blendOption=Linear}
    alphaScaleBiasClamp {bInterpResult=true}
    boneToModify {boneName=dyn_hair_back_01_r}
    rotation=<rot 0.033 0.044 -5.293>
    translation {z=0.944}
   }
   n23 = "Transform (Modify) Bone - Bone: dyn_hair_01_r"(ComponentPose=@n22, Alpha=@n21, Translation="0,0,0", Rotation="0,0,0", Scale="1,1,1") {
    rotationMode=BMM_Additive
    translationMode=BMM_Additive
    alphaBoolBlend {blendOption=Linear}
    alphaScaleBiasClamp {bInterpResult=true}
    boneToModify {boneName=dyn_hair_01_r}
   }
   n24 = "Transform (Modify) Bone - Bone: dyn_hair_back_02_r"(ComponentPose=@n23, Alpha=@n21, Translation="0.169,-2.298,2.588", Rotation="-8.321,-1.595,14.393", Scale="1,1,1") {
    rotationMode=BMM_Additive
    translationMode=BMM_Additive
    alphaBoolBlend {blendOption=Linear}
    alphaScaleBiasClamp {bInterpResult=true}
    boneToModify {boneName=dyn_hair_back_02_r}
    rotation=<rot -8.321 -1.595 14.393>
    translation=<0.169 -2.298 2.588>
   }
   n25 = "Transform (Modify) Bone - Bone: dyn_hair_02_r"(ComponentPose=@n24, Alpha=@n21, Translation="0,0,0", Rotation="0,0,0", Scale="1,1,1") {
    rotationMode=BMM_Additive
    translationMode=BMM_Additive
    alphaBoolBlend {blendOption=Linear}
    alphaScaleBiasClamp {bInterpResult=true}
    boneToModify {boneName=dyn_hair_02_r}
   }
   n26 = "Transform (Modify) Bone - Bone: dyn_hair_back_03_r"(ComponentPose=@n25, Alpha=@n21, Translation="0.735,0.115,2.009", Rotation="2.530,0.210,-3.738", Scale="1,1,1") {
    rotationMode=BMM_Additive
    translationMode=BMM_Additive
    alphaBoolBlend {blendOption=Linear}
    alphaScaleBiasClamp {bInterpResult=true}
    boneToModify {boneName=dyn_hair_back_03_r}
    rotation=<rot 2.53 0.21 -3.738>
    translation=<0.735 0.115 2.009>
   }
   n27 = "Transform (Modify) Bone - Bone: dyn_hair_03_r"(ComponentPose=@n26, Alpha=@n21, Translation="0,0,0", Rotation="0,0,0", Scale="1,1,1") {
    rotationMode=BMM_Additive
    translationMode=BMM_Additive
    alphaBoolBlend {blendOption=Linear}
    alphaScaleBiasClamp {bInterpResult=true}
    boneToModify {boneName=dyn_hair_03_r}
   }
   n28 = "Get (a copy)"(Array=Get head_up_massive(), Dimension 1=2)
   n29 = "Transform (Modify) Bone - Bone: dyn_hair_back_01_r"(ComponentPose=@n27, Alpha=@n28, Translation="-0.351,0.000,0.909", Rotation="3.645,-0.073,-1.632", Scale="1,1,1") {
    rotationMode=BMM_Additive
    translationMode=BMM_Additive
    alphaBoolBlend {blendOption=Linear}
    alphaScaleBiasClamp {bInterpResult=true, interpSpeedDecreasing=30}
    boneToModify {boneName=dyn_hair_back_01_r}
    rotation=<rot 3.645 -0.073 -1.632>
    translation {x=-0.351, z=0.909}
   }
   n30 = "Transform (Modify) Bone - Bone: dyn_hair_01_r"(ComponentPose=@n29, Alpha=@n28, Translation="0,0,0", Rotation="8.549,0.516,5.705", Scale="1,1,1") {
    rotationMode=BMM_Additive
    translationMode=BMM_Additive
    alphaBoolBlend {blendOption=Linear}
    alphaScaleBiasClamp {bInterpResult=true, interpSpeedDecreasing=30}
    boneToModify {boneName=dyn_hair_01_r}
    rotation=<rot 8.549 0.516 5.705>
   }
   n31 = "Transform (Modify) Bone - Bone: dyn_hair_back_02_r"(ComponentPose=@n30, Alpha=@n28, Translation="0.000,-0.000,3.902", Rotation="-15.486,-4.012,-1.685", Scale="1,1,1") {
    rotationMode=BMM_Additive
    translationMode=BMM_Additive
    alphaBoolBlend {blendOption=Linear}
    alphaScaleBiasClamp {bInterpResult=true, interpSpeedDecreasing=30}
    boneToModify {boneName=dyn_hair_back_02_r}
    rotation=<rot -15.486 -4.012 -1.685>
    translation {z=3.902}
   }
   n32 = "Transform (Modify) Bone - Bone: dyn_hair_02_r"(ComponentPose=@n31, Alpha=@n28, Translation="-0.000,0.000,2.173", Rotation="10.960,-2.009,-18.048", Scale="1,1,1") {
    rotationMode=BMM_Additive
    translationMode=BMM_Additive
    alphaBoolBlend {blendOption=Linear}
    alphaScaleBiasClamp {bInterpResult=true, interpSpeedDecreasing=30}
    boneToModify {boneName=dyn_hair_02_r}
    rotation=<rot 10.96 -2.009 -18.048>
    translation {z=2.173}
   }
   n33 = "Transform (Modify) Bone - Bone: dyn_hair_back_03_r"(ComponentPose=@n32, Alpha=@n28, Translation="-0.000,0.000,1.821", Rotation="-9.922,-1.241,7.171", Scale="1,1,1") {
    rotationMode=BMM_Additive
    translationMode=BMM_Additive
    alphaBoolBlend {blendOption=Linear}
    alphaScaleBiasClamp {bInterpResult=true, interpSpeedDecreasing=30}
    boneToModify {boneName=dyn_hair_back_03_r}
    rotation=<rot -9.922 -1.241 7.171>
    translation {z=1.821}
   }
   n34 = "Transform (Modify) Bone - Bone: dyn_hair_03_r"(ComponentPose=@n33, Alpha=@n28, Translation="0,0,0", Rotation="0,0,0", Scale="1,1,1") {
    rotationMode=BMM_Additive
    translationMode=BMM_Additive
    alphaBoolBlend {blendOption=Linear}
    alphaScaleBiasClamp {bInterpResult=true, interpSpeedDecreasing=30}
    boneToModify {boneName=dyn_hair_03_r}
   }
   n35 = "Get (a copy)"(Array=Get head_up_massive(), Dimension 1=3)
   n36 = "Transform (Modify) Bone - Bone: dyn_hair_back_01_r"(ComponentPose=@n34, Alpha=@n35, Translation="-0.000,0.000,0.792", Rotation="4.452,-2.159,13.106", Scale="1,1,1") {
    rotationMode=BMM_Additive
    translationMode=BMM_Additive
    alphaBoolBlend {blendOption=Linear}
    alphaScaleBiasClamp {bInterpResult=true}
    boneToModify {boneName=dyn_hair_back_01_r}
    rotation=<rot 4.452 -2.159 13.106>
    translation {z=0.792}
   }
   n37 = "Transform (Modify) Bone - Bone: dyn_hair_01_r"(ComponentPose=@n36, Alpha=@n35, Translation="0,0,0", Rotation="0,0,0", Scale="1,1,1") {
    rotationMode=BMM_Additive
    translationMode=BMM_Additive
    alphaBoolBlend {blendOption=Linear}
    alphaScaleBiasClamp {bInterpResult=true}
    boneToModify {boneName=dyn_hair_01_r}
   }
   n38 = "Transform (Modify) Bone - Bone: dyn_hair_back_02_r"(ComponentPose=@n37, Alpha=@n35, Translation="-0.000,-0.000,3.346", Rotation="4.648,-1.365,-17.540", Scale="1,1,1") {
    rotationMode=BMM_Additive
    translationMode=BMM_Additive
    alphaBoolBlend {blendOption=Linear}
    alphaScaleBiasClamp {bInterpResult=true}
    boneToModify {boneName=dyn_hair_back_02_r}
    rotation=<rot 4.648 -1.365 -17.54>
    translation {z=3.346}
   }
   n39 = "Transform (Modify) Bone - Bone: dyn_hair_02_r"(ComponentPose=@n38, Alpha=@n35, Translation="0,0,0", Rotation="-0.864,0.000,-0.000", Scale="1,1,1") {
    rotationMode=BMM_Additive
    translationMode=BMM_Additive
    alphaBoolBlend {blendOption=Linear}
    alphaScaleBiasClamp {bInterpResult=true}
    boneToModify {boneName=dyn_hair_02_r}
    rotation {pitch=-0.864}
   }
   n40 = "Transform (Modify) Bone - Bone: dyn_hair_back_03_r"(ComponentPose=@n39, Alpha=@n35, Translation="0.000,0.000,0.484", Rotation="-10.541,0.000,3.383", Scale="1,1,1") {
    rotationMode=BMM_Additive
    translationMode=BMM_Additive
    alphaBoolBlend {blendOption=Linear}
    alphaScaleBiasClamp {bInterpResult=true}
    boneToModify {boneName=dyn_hair_back_03_r}
    rotation {pitch=-10.541, roll=3.383}
    translation {x=0, z=0.4836}
   }
   n41 = "Transform (Modify) Bone - Bone: dyn_hair_03_r"(ComponentPose=@n40, Alpha=@n35, Translation="0,0,0", Rotation="0,0,0", Scale="1,1,1") {
    rotationMode=BMM_Additive
    translationMode=BMM_Additive
    alphaBoolBlend {blendOption=Linear}
    alphaScaleBiasClamp {bInterpResult=true}
    boneToModify {boneName=dyn_hair_03_r}
   }
   n42 = "Get (a copy)"(Array=Get head_up_massive(), Dimension 1=4)
   n43 = "Transform (Modify) Bone - Bone: dyn_hair_back_01_r"(ComponentPose=@n41, Alpha=@n42, Translation="-0.000,-0.720,1.549", Rotation="-4.836,0.264,-0.362", Scale="1,1,1") {
    rotationMode=BMM_Additive
    translationMode=BMM_Additive
    alphaBoolBlend {blendOption=Linear}
    alphaScaleBiasClamp {bInterpResult=true}
    boneToModify {boneName=dyn_hair_back_01_r}
    rotation=<rot -4.836 0.264 -0.362>
    translation {y=-0.72, z=1.549}
   }
   n44 = "Transform (Modify) Bone - Bone: dyn_hair_01_r"(ComponentPose=@n43, Alpha=@n42, Translation="0,0,0", Rotation="2.779,-9.227,7.256", Scale="1,1,1") {
    rotationMode=BMM_Additive
    translationMode=BMM_Additive
    alphaBoolBlend {blendOption=Linear}
    alphaScaleBiasClamp {bInterpResult=true}
    boneToModify {boneName=dyn_hair_01_r}
    rotation=<rot 2.779 -9.227 7.256>
   }
   n45 = "Transform (Modify) Bone - Bone: dyn_hair_back_02_r"(ComponentPose=@n44, Alpha=@n42, Translation="1.013,-0.663,2.641", Rotation="2.784,-10.717,-18.460", Scale="1,1,1") {
    rotationMode=BMM_Additive
    translationMode=BMM_Additive
    alphaBoolBlend {blendOption=Linear}
    alphaScaleBiasClamp {bInterpResult=true}
    boneToModify {boneName=dyn_hair_back_02_r}
    rotation=<rot 2.784 -10.717 -18.46>
    translation=<1.013 -0.663 2.641>
   }
   n46 = "Transform (Modify) Bone - Bone: dyn_hair_02_r"(ComponentPose=@n45, Alpha=@n42, Translation="-0.437,-1.626,2.893", Rotation="7.109,-19.278,-18.330", Scale="1,1,1") {
    rotationMode=BMM_Additive
    translationMode=BMM_Additive
    alphaBoolBlend {blendOption=Linear}
    alphaScaleBiasClamp {bInterpResult=true}
    boneToModify {boneName=dyn_hair_02_r}
    rotation=<rot 7.109 -19.278 -18.33>
    translation=<-0.437 -1.626 2.893>
   }
   n47 = "Transform (Modify) Bone - Bone: dyn_hair_back_03_r"(ComponentPose=@n46, Alpha=@n42, Translation="-1.002,-1.709,3.986", Rotation="3.391,-0.872,-13.126", Scale="1,1,1") {
    rotationMode=BMM_Additive
    translationMode=BMM_Additive
    alphaBoolBlend {blendOption=Linear}
    alphaScaleBiasClamp {bInterpResult=true}
    boneToModify {boneName=dyn_hair_back_03_r}
    rotation=<rot 3.391 -0.872 -13.126>
    translation=<-1.002 -1.709 3.986>
   }
   n48 = "Transform (Modify) Bone - Bone: dyn_hair_03_r"(ComponentPose=@n47, Alpha=@n42, Translation="-0.000,-0.000,1", Rotation="23.729,0.000,0.000", Scale="1,1,1") {
    rotationMode=BMM_Additive
    translationMode=BMM_Additive
    alphaBoolBlend {blendOption=Linear}
    alphaScaleBiasClamp {bInterpResult=true}
    boneToModify {boneName=dyn_hair_03_r}
    rotation {pitch=23.729}
    translation {z=1}
   }
   n49 = "Get (a copy)"(Array=Get head_up_massive(), Dimension 1=5)
   n50 = "Transform (Modify) Bone - Bone: dyn_hair_back_01_r"(ComponentPose=@n48, Alpha=@n49, Translation="0,0,0", Rotation="-2.215,-0.226,13.402", Scale="1,1,1") {
    rotationMode=BMM_Additive
    translationMode=BMM_Additive
    alphaBoolBlend {blendOption=Linear}
    alphaScaleBiasClamp {bInterpResult=true}
    boneToModify {boneName=dyn_hair_back_01_r}
    rotation=<rot -2.215 -0.226 13.402>
   }
   n51 = "Transform (Modify) Bone - Bone: dyn_hair_01_r"(ComponentPose=@n50, Alpha=@n49, Translation="0,0,0", Rotation="0,0,0", Scale="1,1,1") {
    rotationMode=BMM_Additive
    translationMode=BMM_Additive
    alphaBoolBlend {blendOption=Linear}
    alphaScaleBiasClamp {bInterpResult=true}
    boneToModify {boneName=dyn_hair_01_r}
   }
   n52 = "Transform (Modify) Bone - Bone: dyn_hair_back_02_r"(ComponentPose=@n51, Alpha=@n49, Translation="1.412,-2.167,4.339", Rotation="-5.114,0.867,-24.156", Scale="1,1,1") {
    rotationMode=BMM_Additive
    translationMode=BMM_Additive
    alphaBoolBlend {blendOption=Linear}
    alphaScaleBiasClamp {bInterpResult=true}
    boneToModify {boneName=dyn_hair_back_02_r}
    rotation=<rot -5.114 0.867 -24.156>
    translation=<1.412 -2.167 4.339>
   }
   n53 = "Transform (Modify) Bone - Bone: dyn_hair_02_r"(ComponentPose=@n52, Alpha=@n49, Translation="0,0,0", Rotation="0,0,0", Scale="1,1,1") {
    rotationMode=BMM_Additive
    translationMode=BMM_Additive
    alphaBoolBlend {blendOption=Linear}
    alphaScaleBiasClamp {bInterpResult=true}
    boneToModify {boneName=dyn_hair_02_r}
   }
   n54 = "Transform (Modify) Bone - Bone: dyn_hair_back_03_r"(ComponentPose=@n53, Alpha=@n49, Translation="0.735,-2.134,4.228", Rotation="0.930,0.615,18.018", Scale="1,1,1") {
    rotationMode=BMM_Additive
    translationMode=BMM_Additive
    alphaBoolBlend {blendOption=Linear}
    alphaScaleBiasClamp {bInterpResult=true}
    boneToModify {boneName=dyn_hair_back_03_r}
    rotation=<rot 0.93 0.615 18.018>
    translation=<0.735 -2.134 4.228>
   }
   n55 = "Transform (Modify) Bone - Bone: dyn_hair_03_r"(ComponentPose=@n54, Alpha=@n49, Translation="0,0,0", Rotation="0,0,0", Scale="1,1,1") {
    rotationMode=BMM_Additive
    translationMode=BMM_Additive
    alphaBoolBlend {blendOption=Linear}
    alphaScaleBiasClamp {bInterpResult=true}
    boneToModify {boneName=dyn_hair_03_r}
   }
   n56 = "Transform (Modify) Bone - Bone: dyn_hair_02"(ComponentPose=@n55, Alpha=@n49, Translation="0,0,0", Rotation="8.077,-0.560,29.798", Scale="1,1,1") {
    rotationMode=BMM_Additive
    translationMode=BMM_Additive
    alphaBoolBlend {blendOption=Linear}
    alphaScaleBiasClamp {bInterpResult=true}
    boneToModify {boneName=dyn_hair_02}
    rotation=<rot 8.077 -0.56 29.798>
   }
   n57 = "Transform (Modify) Bone - Bone: dyn_hair_03"(ComponentPose=@n56, Alpha=@n49, Translation="0,0,0", Rotation="0,0,0", Scale="1,1,1") {
    rotationMode=BMM_Additive
    translationMode=BMM_Additive
    alphaBoolBlend {blendOption=Linear}
    alphaScaleBiasClamp {bInterpResult=true}
    boneToModify {boneName=dyn_hair_03}
   }
   Output Pose(Result=Component To Local(ComponentPose=@n57)) {layerGroup=None}
   n58 = RigidBody!(
    ComponentPose=Local To Component(LocalPose=Copy Pose From Mesh() {bUseAttachedParent=true})
    Alpha
    OverrideWorldGravity:Vector=Get gravity() -> (gravity:Vector)
    ExternalForce
   ) -> (Pose) {
    bOverrideWorldGravity=true
    overridePhysicsAsset=&/F_MED_Cavalry_Alt_FaceAcc_Physics1
    simulationSpace=BaseBoneSpace
    alphaBoolBlend {blendOption=Linear}
    baseBoneRef {boneName=head}
    componentAppliedLinearAccClamp=<2000 2000 2000>
    componentLinearAccScale=<0.4 0.4 0.4>
    componentLinearVelScale=<0.4 0.4 0.4>
    overrideWorldGravity {z=-1500}
    simSpaceSettings {maxAngularAcceleration=3000, maxAngularVelocity=3000, maxLinearAcceleration=3000, maxLinearVelocity=3000}
   }
   Save cached pose 'SavedPose15620'(Pose=Component To Local(ComponentPose=@n58)) {cacheName=SavedPose15620}
  ]}
 EventGraph {nodes [
   Event BlueprintUpdateAnimation()
   > n1 = Gravity Override State Function DEPRECATED(CharacterPartAnimInstance=Self-Reference(), GameTimeInSeconds=GetGameTimeInSeconds(), GravityOverrideParams=Get Gravity Override Params())
   > Set gravity(gravity:Vector=@n1) -> (Output_Get:Vector)
   > n2 = Switch on EFortWeaponCoreAnimation!(execute, Selection:EFortWeaponCoreAnimation Enum=Get WeaponCoreAnim!() -> (WeaponCoreAnim)) -> (
    Melee:Exec
    Pistol:Exec
    Shotgun:Exec
    PaperBlueprint:Exec
    Rifle:Exec
    MeleeOneHand:Exec
    MachinePistol:Exec
    RocketLauncher:Exec
    GrenadeLauncher:Exec
    GoingCommando:Exec
    AssaultRifle:Exec
    TacticalShotgun:Exec
    SniperRifle:Exec
    TrapPlacement:Exec
    ShoulderLauncher:Exec
    AbilityDecoTool:Exec
    Crossbow:Exec
    C4:Exec
    RemoteControl:Exec
    DualWield:Exec
    AR_BullPup:Exec
    AR_ForwardGrip:Exec
    MedPackPaddles:Exec
    SMG_P90:Exec
    AR_DrumGun:Exec
    Consumable_Small:Exec
    Consumable_Large:Exec
    Balloon:Exec
    MountedTurret:Exec
    CreativeTool:Exec
    ExplosiveBow:Exec
    AshtonIndigo:Exec
    AshtonChicago:Exec
    MeleeDualWield:Exec
    Unarmed:Exec
   )
   n3 = Branch(execute=[@n2.AssaultRifle, @n2.SMG_P90, @n2.ShoulderLauncher, @n2.SniperRifle, @n2.TacticalShotgun], Condition=Get bIsTargeting())
   > n4 = Branch(Condition=Get bIsCrouching())
   > n5 = Branch(Condition=Get bIsMoving2D())
   > Set head_up_massive(head_up_massive:"Array of Float (double-precision)s"=Make Array!("[0]":double=0, "[1]":double=0, "[2]":double=0, "[3]":double=0, "[4]":double=0, "[5]":double=0) -> (Array:"Array of Float (double-precision)s")) -> (Output_Get:"Array of Float (double-precision)s")
   n6 = GetOwnerSkeletalMeshComponent()
   Set head_up_massive(
    execute=@n5.else
    head_up_massive=Make Array(
     "[0]"=0
     "[1]"=0
     "[2]"=0
     "[3]"=0
     "[4]"=0
     "[5]"=GetAlphaFromReferenceTransformDelta(SkeletalMeshComponent=@n6, BoneName=head, BaseName=spine_05, DeltaComponentType=EulerY, InputRange=(X=18,Y=51), TwistAxis="1,0,0")
    )
   )
   Set head_up_massive(
    execute=@n4.else
    head_up_massive=Make Array(
     "[0]"=0
     "[1]"=GetAlphaFromReferenceTransformDelta(SkeletalMeshComponent=@n6, BoneName=head, BaseName=spine_05, DeltaComponentType=EulerY, InputRange=(X=10,Y=43), TwistAxis="1,0,0")
     "[2]"=0
     "[3]"=0
     "[4]"=0
     "[5]"=0
    )
   )
   n7 = Branch(execute=@n3.else, Condition=Get bIsRelaxedLevel1() -> (bIsRelaxedLevel1:bool))
   > Set head_up_massive(head_up_massive=Make Array("[0]"=0, "[1]"=0, "[2]"=0, "[3]"=0, "[4]"=0, "[5]"=0))
   n8 = GetOwnerSkeletalMeshComponent()
   Set head_up_massive(
    execute=@n7.else
    head_up_massive=Make Array(
     "[0]"=0
     "[1]"=0
     "[2]"=0
     "[3]"=GetAlphaFromReferenceTransformDelta(SkeletalMeshComponent=@n8, BoneName=head, BaseName=spine_05, InputRange=(X=-10,Y=-40), TwistAxis="1,0,0")
     "[4]"=0
     "[5]"=0
    )
   )
   n9 = Branch(execute=@n2.Pistol, Condition=Get bIsTargeting())
   > n10 = Branch(Condition=Get bIsCrouching())
   > n11 = Branch(Condition=Get bIsMoving2D())
   > Set head_up_massive(head_up_massive=Make Array("[0]"=0, "[1]"=0, "[2]"=0, "[3]"=0, "[4]"=0, "[5]"=0))
   n12 = GetOwnerSkeletalMeshComponent()
   Set head_up_massive(
    execute=@n11.else
    head_up_massive=Make Array(
     "[0]"=0
     "[1]"=0
     "[2]"=0
     "[3]"=0
     "[4]"=GetAlphaFromReferenceTransformDelta(SkeletalMeshComponent=@n12, BoneName=head, BaseName=spine_05, InputRange=(X=-10,Y=-56), TwistAxis="1,0,0")
     "[5]"=0
    )
   )
   Set head_up_massive(
    execute=@n10.else
    head_up_massive=Make Array(
     "[0]"=GetAlphaFromReferenceTransformDelta(SkeletalMeshComponent=@n12, BoneName=head, BaseName=spine_05, InputRange=(X=-10,Y=-48), TwistAxis="1,0,0")
     "[1]"=0
     "[2]"=0
     "[3]"=0
     "[4]"=0
     "[5]"=0
    )
   )
   n13 = Branch(execute=@n9.else, Condition=Get bIsRelaxedLevel1())
   > Set head_up_massive(head_up_massive=Make Array("[0]"=0, "[1]"=0, "[2]"=0, "[3]"=0, "[4]"=0, "[5]"=0))
   n14 = GetOwnerSkeletalMeshComponent()
   Set head_up_massive(
    execute=@n13.else
    head_up_massive=Make Array(
     "[0]"=0
     "[1]"=0
     "[2]"=GetAlphaFromReferenceTransformDelta(SkeletalMeshComponent=@n14, BoneName=head, BaseName=spine_05, InputRange=(X=-10,Y=-55), TwistAxis="1,0,0")
     "[3]"=0
     "[4]"=0
     "[5]"=0
    )
   )
  ]}
 OnRep_Selection {nodes [OnRep_Selection() {is=title K2Node_FunctionEntry}]}
]
variables [
 gravity {type=Vector}
 Gravity Override Params {
  type=GravityOverrideParamsStruct
  default {
   emoteOrMelee {gravityOverride=<0 0 -1500>}
   front_End {gravityOverride=<0 0 -2000>}
   onGround_Falling {gravityOverride=<0 0 -2000>}
   onGround_Moving {gravityOverride=<0 0 -2000>}
   onGround_Standing {gravityOverride=<0 0 -2000>}
   skydive_FloatingDown {gravityOverride=<0 -700 1500>, windAmplitude=<200 -500 600>}
   skydive_Parachuting {gravityOverride=<0 0 -1000>, windAmplitude=<100 -100 100>, windFrequency=<1.5 3 3>}
  }
 }
 head_up_massive {type="array<number>", default [0,0,0,0,0,0]}
]
